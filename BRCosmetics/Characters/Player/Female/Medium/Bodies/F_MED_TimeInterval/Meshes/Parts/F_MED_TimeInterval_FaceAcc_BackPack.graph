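# Physics asset for the TimeInterval female medium outfit's FaceAcc_BackPack mesh part, providing dynamic hair bodies and upper-body collision shapes.
# Capsules cover head, neck, upper spine, clavicles and upper arms, joined by free-rotation constraints.
# Front, right-side and rear hair chains hang from head through master bones.
# Front and side links have angular limits, rear links rotate freely, and a free constraint cross-links the front and side chains.
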PhysicsAsset /BRCosmetics/Characters/Player/Female/Medium/Bodies/F_MED_TimeInterval/Meshes/Parts/F_MED_TimeInterval_FaceAcc_BackPack
bodies {
 clavicle_l [{shapeType=Capsule, length=7, radius=5.8, shapeName=clavicle_l_capsule, center=<5.8087 0 0>, rotation=<rot -84.8662 0.0019 -0.0019>}]
 clavicle_r [{
   shapeType=Capsule
   length=9
   radius=6.7
   shapeName=clavicle_r_capsule
   center=<-5.6958 0.7275 0.1321>
   rotation=<rot -85.8827 3.6657 -0.0028>
  }]
 dyn_hair_F_01 [{shapeType=Capsule, length=7, radius=2, shapeName=dyn_hair_F_01_capsule, center=<4 0 1>, rotation=<rot 90 0 0>}]
 dyn_hair_F_02 [{shapeType=Capsule, length=4.5, radius=1, shapeName=dyn_hair_F_02_capsule, center=<2.6665 0 0>, rotation=<rot 90 0 0>}]
 dyn_hair_F_03 [{shapeType=Capsule, length=4, radius=1, shapeName=dyn_hair_F_03_capsule, center=<2 0 0>, rotation=<rot 90 0 0>}]
 dyn_hair_F_04 [{shapeType=Capsule, length=4, radius=1.5, shapeName=dyn_hair_F_03_capsule, center=<2 0 0>, rotation=<rot 90 0 0>}]
 dyn_hair_F_mstr [{shapeType=Capsule, length=0.505, radius=0.505, shapeName=dyn_hair_F_mstr_capsule, rotation=<rot 0 -90 0>}]
 dyn_hair_R_01 [{shapeType=Capsule, length=1, radius=1, shapeName=dyn_hair_R_01_capsule}]
 dyn_hair_R_02 [{shapeType=Capsule, length=1, radius=1, shapeName=dyn_hair_R_02_capsule}]
 dyn_hair_R_03 [{shapeType=Capsule, length=6, radius=2, shapeName=dyn_hair_R_03_capsule, rotation=<rot 0 0 90>}]
 dyn_hair_R_mstr [{shapeType=Capsule, length=1, radius=1, shapeName=dyn_hair_R_mstr_capsule}]
 dyn_hair_SR_01 [{shapeType=Capsule, length=4, radius=2.5, shapeName=dyn_hair_SL_01_capsule, center=<3 0 1>, rotation=<rot 90 0 0>}]
 dyn_hair_SR_mstr [{shapeType=Capsule, length=1, radius=1, shapeName=dyn_hair_SR_mstr_capsule}]
 head [{shapeType=Capsule, length=13.2681, radius=6.263, rotation=<rot 90 0 0>}]
 neck_01 [{shapeType=Capsule, length=6, radius=3, shapeName=neck_01_capsule, center=<2.2342 0 0>, rotation=<rot 90 0 0>}]
 spine_03 [{shapeType=Capsule, length=10, radius=7, shapeName=spine_05_capsule, center=<0 -0.777 0>, rotation=<rot 0 -180 180>}]
 spine_04 [{shapeType=Capsule, length=10, radius=7, shapeName=spine_05_capsule, center=<1.7332 -0.777 0>, rotation=<rot 0 -180 180>}]
 spine_05 [{shapeType=Capsule, length=22, radius=7.5, shapeName=spine_05_capsule, center=<2.2482 -0.5 0>, rotation=<rot 0 -180 180>}]
 upperarm_l [{shapeType=Capsule, length=20, radius=4, shapeName=upperarm_r_capsule, center=<15 0 0>, rotation=<rot 90 0 0>}]
 upperarm_r [{shapeType=Capsule, length=20, radius=4, shapeName=upperarm_r_capsule, center=<-15 0 0>, rotation=<rot 90 0 0>}]
}
constraints [
 {
  bone1Name=head
  bone2Name=neck_01
  swing1LimitDegrees=45
  swing1Motion=Free
  swing2LimitDegrees=45
  swing2Motion=Free
  twistLimitDegrees=45
  twistMotion=Free
 }
 {
  bone1Name=clavicle_l
  bone2Name=spine_05
  swing1LimitDegrees=45
  swing1Motion=Free
  swing2LimitDegrees=45
  swing2Motion=Free
  twistLimitDegrees=45
  twistMotion=Free
 }
 {
  bone1Name=clavicle_r
  bone2Name=spine_05
  swing1LimitDegrees=45
  swing1Motion=Free
  swing2LimitDegrees=45
  swing2Motion=Free
  twistLimitDegrees=45
  twistMotion=Free
 }
 {
  bone1Name=spine_05
  bone2Name=spine_04
  swing1LimitDegrees=45
  swing1Motion=Free
  swing2LimitDegrees=45
  swing2Motion=Free
  twistLimitDegrees=45
  twistMotion=Free
 }
 {
  bone1Name=upperarm_r
  bone2Name=clavicle_r
  swing1LimitDegrees=45
  swing1Motion=Free
  swing2LimitDegrees=45
  swing2Motion=Free
  twistLimitDegrees=45
  twistMotion=Free
 }
 {
  bone1Name=dyn_hair_F_mstr
  bone2Name=head
  swing1LimitDegrees=45
  swing1Motion=Free
  swing2LimitDegrees=45
  swing2Motion=Free
  twistLimitDegrees=45
  twistMotion=Free
 }
 {
  bone1Name=dyn_hair_SR_mstr
  bone2Name=head
  swing1LimitDegrees=45
  swing1Motion=Free
  swing2LimitDegrees=45
  swing2Motion=Free
  twistLimitDegrees=45
  twistMotion=Free
 }
 {bone1Name=dyn_hair_SR_01, bone2Name=dyn_hair_SR_mstr, swing1LimitDegrees=65, swing2LimitDegrees=7, twistLimitDegrees=2}
 {bone1Name=dyn_hair_F_01, bone2Name=dyn_hair_F_mstr, swing1LimitDegrees=15, swing2LimitDegrees=5, twistLimitDegrees=5}
 {bone1Name=dyn_hair_F_02, bone2Name=dyn_hair_F_01, swing1LimitDegrees=10, swing2LimitDegrees=5, twistLimitDegrees=2}
 {bone1Name=dyn_hair_F_03, bone2Name=dyn_hair_F_02, swing1LimitDegrees=45, swing2LimitDegrees=15, twistLimitDegrees=2}
 {
  bone1Name=spine_04
  bone2Name=spine_03
  swing1LimitDegrees=45
  swing1Motion=Free
  swing2LimitDegrees=45
  swing2Motion=Free
  twistLimitDegrees=45
  twistMotion=Free
 }
 {
  bone1Name=dyn_hair_F_01
  bone2Name=dyn_hair_SR_01
  swing1LimitDegrees=45
  swing1Motion=Free
  swing2LimitDegrees=45
  swing2Motion=Free
  twistLimitDegrees=45
  twistMotion=Free
 }
 {bone1Name=dyn_hair_F_04, bone2Name=dyn_hair_F_03, swing1LimitDegrees=45, swing2LimitDegrees=45, twistLimitDegrees=2}
 {
  bone1Name=neck_01
  bone2Name=spine_05
  swing1LimitDegrees=45
  swing1Motion=Free
  swing2LimitDegrees=45
  swing2Motion=Free
  twistLimitDegrees=45
  twistMotion=Free
 }
 {
  bone1Name=upperarm_l
  bone2Name=clavicle_l
  swing1LimitDegrees=45
  swing1Motion=Free
  swing2LimitDegrees=45
  swing2Motion=Free
  twistLimitDegrees=45
  twistMotion=Free
 }
 {
  bone1Name=dyn_hair_R_mstr
  bone2Name=head
  swing1LimitDegrees=45
  swing1Motion=Free
  swing2LimitDegrees=45
  swing2Motion=Free
  twistLimitDegrees=45
  twistMotion=Free
 }
 {
  bone1Name=dyn_hair_R_01
  bone2Name=dyn_hair_R_mstr
  swing1LimitDegrees=45
  swing1Motion=Free
  swing2LimitDegrees=45
  swing2Motion=Free
  twistLimitDegrees=45
  twistMotion=Free
 }
 {
  bone1Name=dyn_hair_R_02
  bone2Name=dyn_hair_R_01
  swing1LimitDegrees=45
  swing1Motion=Free
  swing2LimitDegrees=45
  swing2Motion=Free
  twistLimitDegrees=45
  twistMotion=Free
 }
 {
  bone1Name=dyn_hair_R_03
  bone2Name=dyn_hair_R_02
  swing1LimitDegrees=45
  swing1Motion=Free
  swing2LimitDegrees=45
  swing2Motion=Free
  twistLimitDegrees=45
  twistMotion=Free
 }
]
