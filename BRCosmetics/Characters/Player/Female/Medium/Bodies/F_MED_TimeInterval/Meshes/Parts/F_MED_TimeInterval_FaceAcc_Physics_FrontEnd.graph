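# Front-end physics asset for the TimeInterval female character's face accessory, providing dynamic hair physics with capsule bodies on the hair, head and upper body.
# Six hair chains (back, front, left, right and two SL/SR strands) hang from master bones on the head, with per-joint swing and twist limits.
# Angularly free constraints connect the torso, neck and arm bodies, and cross-link neighboring hair chains, some in both directions.
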PhysicsAsset /BRCosmetics/Characters/Player/Female/Medium/Bodies/F_MED_TimeInterval/Meshes/Parts/F_MED_TimeInterval_FaceAcc_Physics_FrontEnd
bodies {
 clavicle_l [{shapeType=Capsule, length=7, radius=5, shapeName=clavicle_l_capsule, center=<5.8087 0 0>, rotation=<rot -84.8662 0.0019 -0.0019>}]
 clavicle_r [{shapeType=Capsule, length=9, radius=5, shapeName=clavicle_r_capsule, center=<-5.168 -0.1327 -2.0489>, rotation=<rot -85.5561 3.6641 -0.0012>}]
 dyn_hair_B_01 [{shapeType=Capsule, length=4.5, radius=4.5, shapeName=dyn_hair_B_01_capsule, center=<5 0 3>, rotation=<rot 0 -90.1331 90.0002>}]
 dyn_hair_B_02 [{shapeType=Capsule, length=4.5, radius=4.5, shapeName=dyn_hair_B_02_capsule, center=<5 0 3>, rotation=<rot 0 -90.506 90.0004>}]
 dyn_hair_B_03 [{shapeType=Capsule, length=4.5, radius=4.5, shapeName=dyn_hair_B_03_capsule, center=<5 0 3>, rotation=<rot 0 90.4154 89.9998>}]
 dyn_hair_B_04 [{shapeType=Capsule, length=5, radius=4, shapeName=dyn_hair_B_04_capsule, center=<5 0 3>, rotation=<rot 0 90.4154 89.9999>}]
 dyn_hair_B_05 [{shapeType=Capsule, length=5, radius=4, shapeName=dyn_hair_B_05_capsule, center=<5 0 3>, rotation=<rot 0 90.4154 89.9999>}]
 dyn_hair_B_mstr [{shapeType=Capsule, length=1, radius=1, shapeName=dyn_hair_B_mstr_capsule}]
 dyn_hair_F_01 [{shapeType=Capsule, length=7, radius=3, shapeName=dyn_hair_F_01_capsule, center=<4 0 0>, rotation=<rot 90 0 0>}]
 dyn_hair_F_02 [{shapeType=Capsule, length=9, radius=2.5, shapeName=dyn_hair_F_02_capsule, center=<2.6665 0 0>, rotation=<rot 90 0 0>}]
 dyn_hair_F_03 [{shapeType=Capsule, length=9, radius=2.5, shapeName=dyn_hair_F_03_capsule, center=<2 0 0>, rotation=<rot 90 0 0>}]
 dyn_hair_F_04 [{shapeType=Capsule, length=4, radius=2.5, shapeName=dyn_hair_F_03_capsule, center=<1 0 0>, rotation=<rot 90 0 0>}]
 dyn_hair_F_mstr [{shapeType=Capsule, length=0.505, radius=0.505, shapeName=dyn_hair_F_mstr_capsule, rotation=<rot 0 -90 0>}]
 dyn_hair_L_01 [{shapeType=Capsule, length=5, radius=3, shapeName=dyn_hair_L_01_capsule, center=<5 0 3>, rotation=<rot 90 0 0>}]
 dyn_hair_L_02 [{shapeType=Capsule, length=5, radius=2, shapeName=dyn_hair_L_02_capsule, center=<5 0 3>, rotation=<rot -8.2651 -84.7345 69.9366>}]
 dyn_hair_L_03 [{shapeType=Capsule, length=5, radius=3.5, shapeName=dyn_hair_L_03_capsule, center=<5 0 3>, rotation=<rot 0.398 91.7508 109.638>}]
 dyn_hair_L_04 [{shapeType=Capsule, length=6, radius=3.5, shapeName=dyn_hair_L_04_capsule, center=<5 0 3.5>, rotation=<rot 0 90.1157 89.9999>}]
 dyn_hair_L_05 [{shapeType=Capsule, length=7, radius=5, shapeName=dyn_hair_L_05_capsule, center=<5 0 3>, rotation=<rot 0 90.1157 89.9999>}]
 dyn_hair_L_06 [{shapeType=Capsule, length=5, radius=4, shapeName=dyn_hair_L_06_capsule, center=<5 0 3>, rotation=<rot 10.1978 90.1176 90.0207>}]
 dyn_hair_L_mstr [{shapeType=Capsule, length=1, radius=1, shapeName=dyn_hair_L_mstr_capsule}]
 dyn_hair_R_01 [{shapeType=Capsule, length=7, radius=4, shapeName=dyn_hair_R_01_capsule, center=<5 0 2>, rotation=<rot 78.6425 -180 -180>}]
 dyn_hair_R_02 [{shapeType=Capsule, length=5, radius=4, shapeName=dyn_hair_F_03_capsule, center=<5 0 3>, rotation=<rot -4.4068 90.1915 102.9038>}]
 dyn_hair_R_03 [{shapeType=Capsule, length=6, radius=4, shapeName=dyn_hair_R_03_capsule, center=<5 0 3.5>, rotation=<rot -7.4178 88.2448 74.8323>}]
 dyn_hair_R_04 [{shapeType=Capsule, length=5, radius=4, shapeName=dyn_hair_R_04_capsule, center=<5 0 3>, rotation=<rot -4.4062 90.9912 89.9237>}]
 dyn_hair_R_05 [{shapeType=Capsule, length=5, radius=4.3, shapeName=dyn_hair_R_05_capsule, center=<5 0 2>, rotation=<rot -4.4062 90.9912 89.9237>}]
 dyn_hair_R_mstr [{shapeType=Capsule, length=1, radius=1, shapeName=dyn_hair_R_mstr_capsule}]
 dyn_hair_SL_01 [{shapeType=Capsule, length=4, radius=3, shapeName=dyn_hair_SL_01_capsule, center=<3 0 0>, rotation=<rot 90 0 0>}]
 dyn_hair_SL_02 [{shapeType=Capsule, length=4, radius=3, shapeName=dyn_hair_SL_02_capsule, center=<3 0 0>, rotation=<rot 90 0 0>}]
 dyn_hair_SL_03 [{shapeType=Capsule, length=6, radius=4, shapeName=dyn_hair_SL_03_capsule, center=<2 0 0>, rotation=<rot 90 168.9434 -134.2469>}]
 dyn_hair_SL_mstr [{shapeType=Capsule, length=1, radius=1, shapeName=dyn_hair_SL_mstr_capsule}]
 dyn_hair_SR_01 [{shapeType=Capsule, length=4, radius=2.5, shapeName=dyn_hair_SL_01_capsule, center=<3 0 1>, rotation=<rot 90 0 0>}]
 dyn_hair_SR_02 [{shapeType=Capsule, length=6, radius=4, shapeName=dyn_hair_SL_02_capsule, center=<3 -0.5 1>, rotation=<rot 90 -122.1063 -131.112>}]
 dyn_hair_SR_03 [{shapeType=Capsule, length=10, radius=2, shapeName=dyn_hair_SL_03_capsule, center=<2.5 0.5 0>, rotation=<rot 90 -82.9835 -99.7832>}]
 dyn_hair_SR_mstr [{shapeType=Capsule, length=1, radius=1, shapeName=dyn_hair_SR_mstr_capsule}]
 head [{shapeType=Capsule, length=9, radius=6.263, center=<1.8158 0 0>, rotation=<rot 90 0 0>}]
 neck_01 [{shapeType=Capsule, length=6, radius=4, shapeName=neck_01_capsule, center=<2.2342 0 0>, rotation=<rot 90 0 0>}]
 pelvis [{shapeType=Capsule, length=11.5, radius=13, shapeName=pelvis_capsule, rotation=<rot 0 -90 0>}]
 spine_03 [{shapeType=Capsule, length=10, radius=7, shapeName=spine_05_capsule, center=<0 -0.777 0>, rotation=<rot 0 -180 180>}]
 spine_04 [{shapeType=Capsule, length=10, radius=7, shapeName=spine_05_capsule, center=<1.7332 -0.777 0>, rotation=<rot 0 -180 180>}]
 spine_05 [{shapeType=Capsule, length=22, radius=7.5, shapeName=spine_05_capsule, center=<2.2482 -0.5 0>, rotation=<rot 0 -180 180>}]
 upperarm_l [{shapeType=Capsule, length=20, radius=4, shapeName=upperarm_r_capsule, center=<15 0 0>, rotation=<rot 90 0 0>}]
 upperarm_r [{shapeType=Capsule, length=24, radius=5, shapeName=upperarm_r_capsule, center=<-15 -2 -2.7581>, rotation=<rot 90 0 0>}]
}
constraints [
 {
  bone1Name=dyn_hair_B_mstr
  bone2Name=head
  swing1LimitDegrees=45
  swing1Motion=Free
  swing2LimitDegrees=45
  swing2Motion=Free
  twistLimitDegrees=45
  twistMotion=Free
 }
 {bone1Name=dyn_hair_B_01, bone2Name=dyn_hair_B_mstr, swing1LimitDegrees=30, swing2LimitDegrees=65, twistLimitDegrees=2}
 {bone1Name=dyn_hair_B_02, bone2Name=dyn_hair_B_01, swing1LimitDegrees=35, swing2LimitDegrees=70, twistLimitDegrees=5}
 {bone1Name=dyn_hair_B_03, bone2Name=dyn_hair_B_02, swing1LimitDegrees=35, swing2LimitDegrees=55, twistLimitDegrees=10}
 {bone1Name=dyn_hair_B_04, bone2Name=dyn_hair_B_03, swing1LimitDegrees=25, swing2LimitDegrees=55, twistLimitDegrees=10}
 {bone1Name=dyn_hair_B_05, bone2Name=dyn_hair_B_04, swing1LimitDegrees=25, swing2LimitDegrees=55, twistLimitDegrees=10}
 {
  bone1Name=dyn_hair_L_mstr
  bone2Name=head
  swing1LimitDegrees=45
  swing1Motion=Free
  swing2LimitDegrees=45
  swing2Motion=Free
  twistLimitDegrees=45
  twistMotion=Free
 }
 {bone1Name=dyn_hair_L_02, bone2Name=dyn_hair_L_01, swing1LimitDegrees=25, swing2LimitDegrees=65, twistLimitDegrees=5}
 {bone1Name=dyn_hair_L_03, bone2Name=dyn_hair_L_02, swing1LimitDegrees=20, swing2LimitDegrees=65, twistLimitDegrees=10}
 {bone1Name=dyn_hair_L_04, bone2Name=dyn_hair_L_03, swing1LimitDegrees=25, swing2LimitDegrees=60, twistLimitDegrees=15}
 {bone1Name=dyn_hair_L_05, bone2Name=dyn_hair_L_04, swing1LimitDegrees=45, swing2LimitDegrees=65, twistLimitDegrees=15}
 {bone1Name=dyn_hair_L_06, bone2Name=dyn_hair_L_05, swing1LimitDegrees=30, swing2LimitDegrees=65, twistLimitDegrees=15}
 {
  bone1Name=dyn_hair_R_mstr
  bone2Name=head
  swing1LimitDegrees=45
  swing1Motion=Free
  swing2LimitDegrees=45
  swing2Motion=Free
  twistLimitDegrees=45
  twistMotion=Free
 }
 {bone1Name=dyn_hair_R_02, bone2Name=dyn_hair_R_01, swing1LimitDegrees=20, swing2LimitDegrees=75, twistLimitDegrees=5}
 {bone1Name=dyn_hair_R_03, bone2Name=dyn_hair_R_02, swing1LimitDegrees=20, swing2LimitDegrees=75, twistLimitDegrees=10}
 {bone1Name=dyn_hair_R_04, bone2Name=dyn_hair_R_03, swing1LimitDegrees=25, swing2LimitDegrees=65, twistLimitDegrees=10}
 {bone1Name=dyn_hair_R_05, bone2Name=dyn_hair_R_04, swing1LimitDegrees=30, swing2LimitDegrees=65, twistLimitDegrees=10}
 {
  bone1Name=head
  bone2Name=neck_01
  swing1LimitDegrees=45
  swing1Motion=Free
  swing2LimitDegrees=45
  swing2Motion=Free
  twistLimitDegrees=45
  twistMotion=Free
 }
 {
  bone1Name=clavicle_l
  bone2Name=spine_05
  swing1LimitDegrees=45
  swing1Motion=Free
  swing2LimitDegrees=45
  swing2Motion=Free
  twistLimitDegrees=45
  twistMotion=Free
 }
 {
  bone1Name=clavicle_r
  bone2Name=spine_05
  swing1LimitDegrees=45
  swing1Motion=Free
  swing2LimitDegrees=45
  swing2Motion=Free
  twistLimitDegrees=45
  twistMotion=Free
 }
 {
  bone1Name=spine_05
  bone2Name=spine_04
  swing1LimitDegrees=45
  swing1Motion=Free
  swing2LimitDegrees=45
  swing2Motion=Free
  twistLimitDegrees=45
  twistMotion=Free
 }
 {
  bone1Name=upperarm_r
  bone2Name=clavicle_r
  swing1LimitDegrees=45
  swing1Motion=Free
  swing2LimitDegrees=45
  swing2Motion=Free
  twistLimitDegrees=45
  twistMotion=Free
 }
 {
  bone1Name=dyn_hair_L_04
  bone2Name=dyn_hair_B_03
  swing1LimitDegrees=45
  swing1Motion=Free
  swing2LimitDegrees=45
  swing2Motion=Free
  twistLimitDegrees=45
  twistMotion=Free
 }
 {bone1Name=dyn_hair_R_01, bone2Name=dyn_hair_R_mstr, swing1LimitDegrees=30, swing2LimitDegrees=65, twistLimitDegrees=5}
 {
  bone1Name=dyn_hair_F_mstr
  bone2Name=head
  swing1LimitDegrees=45
  swing1Motion=Free
  swing2LimitDegrees=45
  swing2Motion=Free
  twistLimitDegrees=45
  twistMotion=Free
 }
 {bone1Name=dyn_hair_L_01, bone2Name=dyn_hair_L_mstr, swing1LimitDegrees=30, swing2LimitDegrees=65, twistLimitDegrees=5}
 {
  bone1Name=dyn_hair_SR_mstr
  bone2Name=head
  swing1LimitDegrees=45
  swing1Motion=Free
  swing2LimitDegrees=45
  swing2Motion=Free
  twistLimitDegrees=45
  twistMotion=Free
 }
 {bone1Name=dyn_hair_SR_01, bone2Name=dyn_hair_SR_mstr, swing1LimitDegrees=75, swing2LimitDegrees=5, twistLimitDegrees=2}
 {bone1Name=dyn_hair_SR_02, bone2Name=dyn_hair_SR_01, swing1LimitDegrees=65, swing2LimitDegrees=6, twistLimitDegrees=2}
 {bone1Name=dyn_hair_SR_03, bone2Name=dyn_hair_SR_02, swing1LimitDegrees=75, swing2LimitDegrees=10, twistLimitDegrees=2}
 {
  bone1Name=dyn_hair_SL_mstr
  bone2Name=head
  swing1LimitDegrees=45
  swing1Motion=Free
  swing2LimitDegrees=45
  swing2Motion=Free
  twistLimitDegrees=45
  twistMotion=Free
 }
 {bone1Name=dyn_hair_SL_01, bone2Name=dyn_hair_SL_mstr, swing1LimitDegrees=35, swing2LimitDegrees=13, twistLimitDegrees=5}
 {bone1Name=dyn_hair_SL_02, bone2Name=dyn_hair_SL_01, swing1LimitDegrees=20, swing2LimitDegrees=13, twistLimitDegrees=5}
 {bone1Name=dyn_hair_SL_03, bone2Name=dyn_hair_SL_02, swing1LimitDegrees=25, swing2LimitDegrees=15, twistLimitDegrees=2}
 {
  bone1Name=dyn_hair_SR_03
  bone2Name=dyn_hair_R_02
  swing1LimitDegrees=45
  swing1Motion=Free
  swing2LimitDegrees=45
  swing2Motion=Free
  twistLimitDegrees=45
  twistMotion=Free
 }
 {bone1Name=dyn_hair_F_01, bone2Name=dyn_hair_F_mstr, swing1LimitDegrees=10, swing2LimitDegrees=2, twistLimitDegrees=5}
 {bone1Name=dyn_hair_F_02, bone2Name=dyn_hair_F_01, swing1LimitDegrees=15, swing2LimitDegrees=10, twistLimitDegrees=5}
 {bone1Name=dyn_hair_F_03, bone2Name=dyn_hair_F_02, swing1LimitDegrees=12, swing2LimitDegrees=10, twistLimitDegrees=5}
 {
  bone1Name=spine_04
  bone2Name=spine_03
  swing1LimitDegrees=45
  swing1Motion=Free
  swing2LimitDegrees=45
  swing2Motion=Free
  twistLimitDegrees=45
  twistMotion=Free
 }
 {
  bone1Name=dyn_hair_R_03
  bone2Name=dyn_hair_B_02
  swing1LimitDegrees=45
  swing1Motion=Free
  swing2LimitDegrees=45
  swing2Motion=Free
  twistLimitDegrees=45
  twistMotion=Free
 }
 {
  bone1Name=dyn_hair_B_04
  bone2Name=dyn_hair_L_05
  swing1LimitDegrees=45
  swing1Motion=Free
  swing2LimitDegrees=45
  swing2Motion=Free
  twistLimitDegrees=45
  twistMotion=Free
 }
 {
  bone1Name=dyn_hair_L_03
  bone2Name=dyn_hair_B_02
  swing1LimitDegrees=45
  swing1Motion=Free
  swing2LimitDegrees=45
  swing2Motion=Free
  twistLimitDegrees=45
  twistMotion=Free
 }
 {
  bone1Name=dyn_hair_B_03
  bone2Name=dyn_hair_L_04
  swing1LimitDegrees=45
  swing1Motion=Free
  swing2LimitDegrees=45
  swing2Motion=Free
  twistLimitDegrees=45
  twistMotion=Free
 }
 {
  bone1Name=dyn_hair_B_03
  bone2Name=dyn_hair_R_04
  swing1LimitDegrees=45
  swing1Motion=Free
  swing2LimitDegrees=45
  swing2Motion=Free
  twistLimitDegrees=45
  twistMotion=Free
 }
 {
  bone1Name=dyn_hair_R_04
  bone2Name=dyn_hair_B_03
  swing1LimitDegrees=45
  swing1Motion=Free
  swing2LimitDegrees=45
  swing2Motion=Free
  twistLimitDegrees=45
  twistMotion=Free
 }
 {
  bone1Name=dyn_hair_R_05
  bone2Name=dyn_hair_B_04
  swing1LimitDegrees=45
  swing1Motion=Free
  swing2LimitDegrees=45
  swing2Motion=Free
  twistLimitDegrees=45
  twistMotion=Free
 }
 {
  bone1Name=dyn_hair_R_05
  bone2Name=dyn_hair_B_03
  swing1LimitDegrees=45
  swing1Motion=Free
  swing2LimitDegrees=45
  swing2Motion=Free
  twistLimitDegrees=45
  twistMotion=Free
 }
 {
  bone1Name=dyn_hair_B_04
  bone2Name=dyn_hair_R_05
  swing1LimitDegrees=45
  swing1Motion=Free
  swing2LimitDegrees=45
  swing2Motion=Free
  twistLimitDegrees=45
  twistMotion=Free
 }
 {
  bone1Name=dyn_hair_B_02
  bone2Name=dyn_hair_R_03
  swing1LimitDegrees=45
  swing1Motion=Free
  swing2LimitDegrees=45
  swing2Motion=Free
  twistLimitDegrees=45
  twistMotion=Free
 }
 {bone1Name=dyn_hair_F_04, bone2Name=dyn_hair_F_03, swing1LimitDegrees=45, swing2LimitDegrees=45, twistLimitDegrees=5}
 {
  bone1Name=neck_01
  bone2Name=spine_05
  swing1LimitDegrees=45
  swing1Motion=Free
  swing2LimitDegrees=45
  swing2Motion=Free
  twistLimitDegrees=45
  twistMotion=Free
 }
 {
  bone1Name=dyn_hair_SL_03
  bone2Name=dyn_hair_L_03
  swing1LimitDegrees=45
  swing1Motion=Free
  swing2LimitDegrees=45
  swing2Motion=Free
  twistLimitDegrees=45
  twistMotion=Free
 }
 {
  bone1Name=upperarm_l
  bone2Name=clavicle_l
  swing1LimitDegrees=45
  swing1Motion=Free
  swing2LimitDegrees=45
  swing2Motion=Free
  twistLimitDegrees=45
  twistMotion=Free
 }
 {
  bone1Name=dyn_hair_L_05
  bone2Name=dyn_hair_B_04
  swing1LimitDegrees=45
  swing1Motion=Free
  swing2LimitDegrees=45
  swing2Motion=Free
  twistLimitDegrees=45
  twistMotion=Free
 }
 {
  bone1Name=spine_03
  bone2Name=pelvis
  swing1LimitDegrees=45
  swing1Motion=Free
  swing2LimitDegrees=45
  swing2Motion=Free
  twistLimitDegrees=45
  twistMotion=Free
 }
]
