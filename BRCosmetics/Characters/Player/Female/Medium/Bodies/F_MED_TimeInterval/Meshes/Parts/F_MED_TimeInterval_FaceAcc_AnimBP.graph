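# Animation Blueprint for the TimeInterval female medium FaceAcc mesh, adding secondary motion to its dyn_hair bones and accessory physics.
# Copy Pose From Mesh follows the body, with weighted constraints and additive bone offsets repositioning hair bones.
# Boolean blends pick between RigidBody simulations using default, cape, front-end and skydive physics assets, or an AnimDynamics hair chain with spherical limits.
# These branches read the PostCtrlRigCache pose; SkyDiveAlpha weights some simulations.
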AnimBlueprint /BRCosmetics/Characters/Player/Female/Medium/Bodies/F_MED_TimeInterval/Meshes/Parts/F_MED_TimeInterval_FaceAcc_AnimBP
target_skeleton=&/F_MED_TimeInterval_FaceAcc_Skeleton
graphs [
 AnimGraph {nodes [
   Anim Dynamics - Bone: dyn_hair_B_01(Alpha=1, ExternalForce="0,0,0") {
    simulationSpace=BoneRelative
    alphaBoolBlend {blendOption=Linear}
    boundBone {boneName=dyn_hair_B_01}
    componentAppliedLinearAccClamp=<10000 10000 10000>
    componentLinearAccScale=<2 2 2>
    componentLinearVelScale=<2 2 2>
    physicsBodyDefinitions [{
      boundBone {boneName=dyn_hair_B_01}
      boxExtents=<10 10 10>
      constraintSetup {angularLimitsMax=<3 5 15>, angularLimitsMin=<-3 -45 -15>, angularTarget=<0 0 0>}
      localJointOffset=<1 0 0>
     }]
    relativeSpaceBone {boneName=head}
   }
   Anim Dynamics - Bone: dyn_hair_B_03(Alpha=1, ExternalForce="0,0,0") {
    angularDampingOverride=1
    angularSpringConstant=10000
    bAngularSpring=true
    bOverrideAngularDamping=true
    bUseGravityOverride=true
    bUseSphericalLimits=true
    gravityScale=0.35
    simulationSpace=BoneRelative
    alphaBoolBlend {blendOption=Linear}
    boundBone {boneName=dyn_hair_B_03}
    componentAppliedLinearAccClamp=<3000 3000 3000>
    componentLinearAccScale=<2 2 2>
    componentLinearVelScale=<2 2 2>
    gravityOverride {z=-980}
    physicsBodyDefinitions [{
      boundBone {boneName=dyn_hair_B_03}
      boxExtents=<10 40 40>
      constraintSetup {angularLimitsMax=<0 20 15>, angularLimitsMin=<0 -45 -15>, angularTarget=<1 0 0>}
      localJointOffset=<3 0 0>
     }]
    relativeSpaceBone {boneName=head}
    sphericalLimits [{limitRadius=12, drivingBone {boneName=spine_05}, sphereLocalOffset=<0 0 0>}]
   }
   n1 = Get Body_Mesh()
   n2 = "Transform (Modify) Bone - Bone: dyn_hair_L_02"(
    ComponentPose=Local To Component(LocalPose=Copy Pose From Mesh(SourceMeshComponent=@n1) {bCopyCurves=true, bUseAttachedParent=true})
    Alpha=1
    Translation="3.034,500,4.061"
    Rotation="-14.279,0.000,-0.000"
    Scale="1,1,1"
   ) {
    rotationMode=BMM_Additive
    rotationSpace=BCS_BoneSpace
    scaleMode=BMM_Additive
    scaleSpace=BCS_BoneSpace
    translationMode=BMM_Additive
    translationSpace=BCS_BoneSpace
    alphaBoolBlend {blendOption=Linear}
    boneToModify {boneName=dyn_hair_L_02}
    rotation {pitch=-14.279}
    translation=<3.034 500 4.061>
   }
   Component To Local(ComponentPose=@n2)
   RigidBody!(ComponentPose, Alpha, OverrideWorldGravity:Vector=Get Gravity Override(), ExternalForce) -> (Pose) {
    bClampLinearTranslationLimitToRefPose=true
    bOverrideWorldGravity=true
    overridePhysicsAsset=&/F_MED_TimeInterval_FaceAcc_BackPack
    simulationSpace=BaseBoneSpace
    alphaBoolBlend {blendOption=Linear}
    alphaScaleBiasClamp {bInterpResult=true, interpSpeedIncreasing=2}
    baseBoneRef {boneName=spine_05}
    componentAppliedLinearAccClamp=<2000 2000 2000>
    componentLinearAccScale=<0.1 0.1 0.6>
    componentLinearVelScale=<0.1 0.1 0.6>
    simSpaceSettings {
     maxAngularAcceleration=1000
     maxAngularVelocity=3
     maxLinearAcceleration=700
     maxLinearVelocity=500
     velocityScaleZ=0.3
     worldAlpha=0.2
     externalLinearDragV=<0.2 0.2>
    }
   }
   RigidBody(OverrideWorldGravity=Get Gravity Override()) {
    bClampLinearTranslationLimitToRefPose=true
    bOverrideWorldGravity=true
    overridePhysicsAsset=&F_MED_TimeInterval_FaceAcc_BackPack
    simulationSpace=BaseBoneSpace
    alphaBoolBlend {blendOption=Linear}
    alphaScaleBiasClamp {bInterpResult=true, interpSpeedIncreasing=2}
    baseBoneRef {boneName=spine_05}
    componentAppliedLinearAccClamp=<2000 2000 2000>
    componentLinearAccScale=<0.1 0.1 0.6>
    componentLinearVelScale=<0.1 0.1 0.6>
    simSpaceSettings {
     maxAngularAcceleration=1000
     maxAngularVelocity=3
     maxLinearAcceleration=700
     maxLinearVelocity=500
     velocityScaleZ=0.3
     worldAlpha=0.2
     externalLinearDragV=<0.2 0.2>
    }
   }
   n3 = Get SkyDiveAlpha() -> (SkyDiveAlpha:double)
   n4 = RigidBody(ComponentPose=Local To Component(LocalPose=Use cached pose 'PostCtrlRigCache'()), Alpha=@n3, OverrideWorldGravity=Get Gravity Override()) {
    bClampLinearTranslationLimitToRefPose=true
    bOverrideWorldGravity=true
    overridePhysicsAsset=&/F_MED_TimeInterval_FaceAcc_Physics
    simulationSpace=BaseBoneSpace
    alphaBoolBlend {blendOption=Linear}
    alphaScaleBiasClamp {bInterpResult=true, interpSpeedIncreasing=2}
    baseBoneRef {boneName=spine_05}
    componentAppliedLinearAccClamp=<4000 4000 4000>
    componentLinearAccScale=<0.33 0.4 0.5>
    componentLinearVelScale=<0.33 0.4 0.5>
    simSpaceSettings {
     maxAngularAcceleration=1000
     maxAngularVelocity=3
     maxLinearAcceleration=700
     maxLinearVelocity=500
     worldAlpha=0.8
     externalLinearDragV=<0.2 0.2>
    }
   }
   n5 = RigidBody(ComponentPose=Local To Component(LocalPose=Use cached pose 'PostCtrlRigCache'()), OverrideWorldGravity=Get Gravity Override()) {
    bClampLinearTranslationLimitToRefPose=true
    bOverrideWorldGravity=true
    overridePhysicsAsset=&/F_MED_TimeInterval_FaceAcc_Physics_FrontEnd
    simulationSpace=BaseBoneSpace
    alphaBoolBlend {blendOption=Linear}
    alphaScaleBiasClamp {bInterpResult=true, interpSpeedIncreasing=2}
    baseBoneRef {boneName=spine_05}
    componentAppliedLinearAccClamp=<4000 4000 4000>
    componentLinearAccScale=<0.33 0.4 0.5>
    componentLinearVelScale=<0.33 0.4 0.5>
    simSpaceSettings {
     maxAngularAcceleration=1000
     maxAngularVelocity=3
     maxLinearAcceleration=700
     maxLinearVelocity=500
     worldAlpha=0.8
     externalLinearDragV=<0.2 0.2>
    }
   }
   n6 = RigidBody(ComponentPose=Local To Component(LocalPose=Use cached pose 'PostCtrlRigCache'()), OverrideWorldGravity=Get Gravity Override()) {
    bClampLinearTranslationLimitToRefPose=true
    bOverrideWorldGravity=true
    overridePhysicsAsset=&/F_MED_TimeInterval_FaceAcc_Physics_SkyDive
    simulationSpace=BaseBoneSpace
    alphaBoolBlend {blendOption=Linear}
    alphaScaleBiasClamp {bInterpResult=true, interpSpeedIncreasing=2}
    baseBoneRef {boneName=spine_05}
    componentAppliedLinearAccClamp=<1000 1000 1000>
    componentLinearAccScale=<0.1 0.1 0.1>
    componentLinearVelScale=<0.1 0.1 0.1>
    simSpaceSettings {
     maxAngularAcceleration=500
     maxAngularVelocity=3
     maxLinearAcceleration=300
     maxLinearVelocity=300
     velocityScaleZ=0
     worldAlpha=0.4
     externalLinearDragV=<0.2 0.2>
    }
   }
   n7 = Blend Poses by bool(
    BlendPose_0=Component To Local(ComponentPose=@n6)
    BlendPose_1=Blend Poses by bool(BlendPose_0=Component To Local(ComponentPose=@n5), BlendPose_1=Component To Local(ComponentPose=@n4)) {blendType=Linear}
   ) {blendType=Linear}
   n8 = RigidBody(ComponentPose=Local To Component(LocalPose=Use cached pose 'PostCtrlRigCache'()), OverrideWorldGravity=Get Gravity Override()) {
    bClampLinearTranslationLimitToRefPose=true
    bOverrideWorldGravity=true
    overridePhysicsAsset=&/F_MED_TimeInterval_FaceAcc_Physics_Cape_FE
    simulationSpace=BaseBoneSpace
    alphaBoolBlend {blendOption=Linear}
    alphaScaleBiasClamp {bInterpResult=true, interpSpeedIncreasing=2}
    baseBoneRef {boneName=spine_05}
    componentAppliedLinearAccClamp=<3000 3000 3000>
    componentLinearAccScale=<0.3 0.4 0.6>
    componentLinearVelScale=<0.3 0.4 0.6>
    simSpaceSettings {
     maxAngularAcceleration=1000
     maxAngularVelocity=3
     maxLinearAcceleration=700
     maxLinearVelocity=500
     velocityScaleZ=0.3
     worldAlpha=0.7
     externalLinearDragV=<0.2 0.2>
    }
   }
   n9 = RigidBody(ComponentPose=Local To Component(LocalPose=Use cached pose 'PostCtrlRigCache'()), Alpha=@n3, OverrideWorldGravity=Get Gravity Override()) {
    bClampLinearTranslationLimitToRefPose=true
    bOverrideWorldGravity=true
    overridePhysicsAsset=&/F_MED_TimeInterval_FaceAcc_Physics_Cape_FrontEnd
    simulationSpace=BaseBoneSpace
    alphaBoolBlend {blendOption=Linear}
    alphaScaleBiasClamp {bInterpResult=true, interpSpeedIncreasing=2}
    baseBoneRef {boneName=spine_05}
    componentAppliedLinearAccClamp=<3000 3000 3000>
    componentLinearAccScale=<0.3 0.4 0.6>
    componentLinearVelScale=<0.3 0.4 0.6>
    simSpaceSettings {
     maxAngularAcceleration=1000
     maxAngularVelocity=3
     maxLinearAcceleration=700
     maxLinearVelocity=500
     velocityScaleZ=0.3
     worldAlpha=0.7
     externalLinearDragV=<0.2 0.2>
    }
   }
   n10 = RigidBody(ComponentPose=Local To Component(LocalPose=Use cached pose 'PostCtrlRigCache'()), OverrideWorldGravity=Get Gravity Override()) {
    bClampLinearTranslationLimitToRefPose=true
    bOverrideWorldGravity=true
    overridePhysicsAsset=&/F_MED_TimeInterval_FaceAcc_Physics_Cape_SkyDive
    simulationSpace=BaseBoneSpace
    alphaBoolBlend {blendOption=Linear}
    alphaScaleBiasClamp {bInterpResult=true, interpSpeedIncreasing=2}
    baseBoneRef {boneName=spine_05}
    componentAppliedLinearAccClamp=<2000 2000 2000>
    componentLinearAccScale=<0.1 0.1 0.6>
    componentLinearVelScale=<0.1 0.1 0.6>
    simSpaceSettings {
     maxAngularAcceleration=1000
     maxAngularVelocity=3
     maxLinearAcceleration=700
     maxLinearVelocity=500
     velocityScaleZ=0.3
     worldAlpha=0.2
     externalLinearDragV=<0.2 0.2>
    }
   }
   n11 = Blend Poses by bool(
    BlendPose_0=Component To Local(ComponentPose=@n10)
    BlendPose_1=Blend Poses by bool(BlendPose_0=Component To Local(ComponentPose=@n8), BlendPose_1=Component To Local(ComponentPose=@n9)) {blendType=Linear}
   ) {blendType=Linear}
   n12 = RigidBody(ComponentPose=Local To Component(LocalPose=Use cached pose 'PostCtrlRigCache'()), Alpha=@n3, OverrideWorldGravity=Get Gravity Override()) {
    bClampLinearTranslationLimitToRefPose=true
    bOverrideWorldGravity=true
    overridePhysicsAsset=&/F_MED_TimeInterval_FaceAcc_specific_cape
    simulationSpace=BaseBoneSpace
    alphaBoolBlend {blendOption=Linear}
    alphaScaleBiasClamp {bInterpResult=true, interpSpeedIncreasing=2}
    baseBoneRef {boneName=spine_05}
    componentAppliedLinearAccClamp=<2500 2500 2500>
    componentLinearAccScale=<0.3 0.4 0.2>
    componentLinearVelScale=<0.3 0.4 0.2>
    simSpaceSettings {
     maxAngularAcceleration=1000
     maxAngularVelocity=3
     maxLinearAcceleration=700
     maxLinearVelocity=500
     worldAlpha=0.7
     externalLinearDragV=<0.2 0.2>
    }
   }
   n13 = Anim Dynamics - Bone: dyn_hair_B_01(ComponentPose=Local To Component(LocalPose=Use cached pose 'PostCtrlRigCache'()), Alpha=1, ExternalForce="0,0,0") {
    angularDampingOverride=1
    angularSpringConstant=10000
    bAngularSpring=true
    bOverrideAngularDamping=true
    bUseGravityOverride=true
    bUsePlanarLimit=false
    bUseSphericalLimits=true
    gravityScale=0.35
    simulationSpace=BoneRelative
    alphaBoolBlend {blendOption=Linear}
    boundBone {boneName=dyn_hair_B_01}
    componentAppliedLinearAccClamp=<3000 3000 3000>
    componentLinearAccScale=<0.05 0.05 0.7>
    componentLinearVelScale=<0.05 0.05 0.7>
    gravityOverride {z=-980}
    physicsBodyDefinitions [{
      boundBone {boneName=dyn_hair_B_01}
      boxExtents=<40 20 20>
      constraintSetup {angularLimitsMax=<0 5 10>, angularLimitsMin=<0 -45 -10>, angularTarget=<1 0 0>}
      localJointOffset=<3 0 0>
     }]
    planarLimits [{drivingBone {boneName=spine_05}, planeTransform {location=<0 -10 0>, rotation=<rot 0 0 -90>}}]
    relativeSpaceBone {boneName=root}
    sphericalLimits [{limitRadius=16, drivingBone {boneName=spine_05}, sphereLocalOffset=<0 5 0>}]
   }
   n14 = Anim Dynamics - Bone: dyn_hair_B_02(ComponentPose=@n13, Alpha=1, ExternalForce="0,0,0") {
    angularDampingOverride=1
    angularSpringConstant=10000
    bAngularSpring=true
    bOverrideAngularDamping=true
    bUseGravityOverride=true
    bUsePlanarLimit=false
    bUseSphericalLimits=true
    gravityScale=0.35
    simulationSpace=BoneRelative
    alphaBoolBlend {blendOption=Linear}
    boundBone {boneName=dyn_hair_B_02}
    componentAppliedLinearAccClamp=<3000 3000 3000>
    componentLinearAccScale=<0.05 0.05 0.7>
    componentLinearVelScale=<0.05 0.05 0.7>
    gravityOverride {z=-980}
    physicsBodyDefinitions [{
      boundBone {boneName=dyn_hair_B_02}
      boxExtents=<40 40 40>
      constraintSetup {angularLimitsMax=<0 5 10>, angularLimitsMin=<0 -45 -10>, angularTarget=<1 0 0>}
      localJointOffset=<3 0 0>
     }]
    planarLimits [{drivingBone {boneName=spine_05}, planeTransform {location=<0 -10 0>, rotation=<rot 0 0 -90>}}]
    relativeSpaceBone {boneName=root}
    sphericalLimits [{limitRadius=19, drivingBone {boneName=spine_05}, sphereLocalOffset=<0 5 0>}]
   }
   n15 = Anim Dynamics - Bone: dyn_hair_B_03(ComponentPose=@n14, Alpha=1, ExternalForce="0,0,0") {
    angularDampingOverride=0.99
    angularSpringConstant=10000
    bOverrideAngularDamping=true
    bUseGravityOverride=true
    bUsePlanarLimit=false
    bUseSphericalLimits=true
    gravityScale=0.35
    simulationSpace=BoneRelative
    alphaBoolBlend {blendOption=Linear}
    boundBone {boneName=dyn_hair_B_03}
    componentAppliedLinearAccClamp=<3000 3000 3000>
    componentLinearAccScale=<2 2 3>
    componentLinearVelScale=<2 2 0.7>
    gravityOverride {z=-980}
    physicsBodyDefinitions [{
      boundBone {boneName=dyn_hair_B_03}
      boxExtents=<40 20 20>
      constraintSetup {angularLimitsMax=<0 5 10>, angularLimitsMin=<0 -45 -10>, angularTarget=<1 0 0>}
      localJointOffset=<3 0 0>
     }]
    planarLimits [{drivingBone {boneName=spine_05}, planeTransform {location=<0 -10 0>, rotation=<rot 0 0 -90>}}]
    relativeSpaceBone {boneName=root}
    sphericalLimits [{limitRadius=12, drivingBone {boneName=spine_05}, sphereLocalOffset=<0 0 0>}]
   }
   n16 = Anim Dynamics - Bone: dyn_hair_B_04(ComponentPose=@n15, Alpha=1, ExternalForce="0,0,0") {
    angularDampingOverride=0.9
    angularSpringConstant=10000
    bOverrideAngularDamping=true
    bUseGravityOverride=true
    bUsePlanarLimit=false
    bUseSphericalLimits=true
    gravityScale=0.35
    simulationSpace=BoneRelative
    alphaBoolBlend {blendOption=Linear}
    boundBone {boneName=dyn_hair_B_04}
    componentAppliedLinearAccClamp=<3000 3000 3000>
    componentLinearAccScale=<1 1 0.7>
    componentLinearVelScale=<1 1 0.7>
    gravityOverride {z=-980}
    physicsBodyDefinitions [{
      boundBone {boneName=dyn_hair_B_04}
      boxExtents=<40 20 20>
      constraintSetup {angularLimitsMax=<0 5 10>, angularLimitsMin=<0 -45 -10>, angularTarget=<1 0 0>}
      localJointOffset=<3 0 0>
     }]
    planarLimits [{drivingBone {boneName=spine_05}, planeTransform {location=<0 -10 0>, rotation=<rot 0 0 -90>}}]
    relativeSpaceBone {boneName=root}
    sphericalLimits [{limitRadius=12, drivingBone {boneName=spine_05}, sphereLocalOffset=<0 -1 0>}]
   }
   n17 = Anim Dynamics - Bone: dyn_hair_B_05(ComponentPose=@n16, Alpha=1, ExternalForce="0,0,0") {
    angularDampingOverride=0.9
    angularSpringConstant=10000
    bOverrideAngularDamping=true
    bUseGravityOverride=true
    bUsePlanarLimit=false
    bUseSphericalLimits=true
    gravityScale=0.35
    simulationSpace=BoneRelative
    alphaBoolBlend {blendOption=Linear}
    boundBone {boneName=dyn_hair_B_05}
    componentAppliedLinearAccClamp=<3000 3000 3000>
    componentLinearAccScale=<1 1 0.7>
    componentLinearVelScale=<1 1 0.7>
    gravityOverride {z=-980}
    physicsBodyDefinitions [{
      boundBone {boneName=dyn_hair_B_05}
      boxExtents=<40 20 20>
      constraintSetup {angularLimitsMax=<0 5 10>, angularLimitsMin=<0 -45 -10>, angularTarget=<1 0 0>}
      localJointOffset=<3 0 0>
     }]
    planarLimits [{drivingBone {boneName=spine_05}, planeTransform {location=<0 -10 0>, rotation=<rot 0 0 -90>}}]
    relativeSpaceBone {boneName=root}
    sphericalLimits [{limitRadius=12, drivingBone {boneName=spine_05}, sphereLocalOffset=<-3 -1 0>}]
   }
   n18 = Anim Dynamics - Bone: dyn_hair_L_02(ComponentPose=@n17, Alpha=1, ExternalForce="0,0,0") {
    angularDampingOverride=1
    angularSpringConstant=10000
    bAngularSpring=true
    bOverrideAngularDamping=true
    bUseGravityOverride=true
    bUsePlanarLimit=false
    bUseSphericalLimits=true
    gravityScale=0.35
    simulationSpace=BoneRelative
    alphaBoolBlend {blendOption=Linear}
    boundBone {boneName=dyn_hair_L_02}
    componentAppliedLinearAccClamp=<3000 3000 3000>
    componentLinearAccScale=<0.1 0.1 0.7>
    componentLinearVelScale=<0.1 0.1 0.7>
    gravityOverride {z=-980}
    physicsBodyDefinitions [{
      boundBone {boneName=dyn_hair_L_02}
      boxExtents=<10 40 40>
      constraintSetup {angularLimitsMax=<0 30 10>, angularLimitsMin=<0 -45 -10>, angularTarget=<1 0 0>}
      localJointOffset=<3 0 0>
     }]
    planarLimits [{drivingBone {boneName=spine_05}, planeTransform {location=<0 -10 0>, rotation=<rot 0 0 -90>}}]
    relativeSpaceBone {boneName=root}
    sphericalLimits [{limitRadius=20, drivingBone {boneName=spine_05}, sphereLocalOffset=<0 0 -5>}]
   }
   n19 = Anim Dynamics - Bone: dyn_hair_L_03(ComponentPose=@n18, Alpha=1, ExternalForce="0,0,0") {
    angularDampingOverride=1
    angularSpringConstant=10000
    bOverrideAngularDamping=true
    bUseGravityOverride=true
    bUsePlanarLimit=false
    bUseSphericalLimits=true
    gravityScale=0.35
    simulationSpace=BoneRelative
    alphaBoolBlend {blendOption=Linear}
    boundBone {boneName=dyn_hair_L_03}
    componentAppliedLinearAccClamp=<3000 3000 3000>
    componentLinearAccScale=<0.5 0.5 0.5>
    componentLinearVelScale=<0.5 0.5 0.5>
    gravityOverride {z=-980}
    physicsBodyDefinitions [{
      boundBone {boneName=dyn_hair_L_03}
      boxExtents=<10 40 40>
      constraintSetup {angularLimitsMax=<0 55 10>, angularLimitsMin=<0 -45 -10>, angularTarget=<1 0 0>}
      localJointOffset=<3 0 0>
     }]
    planarLimits [{drivingBone {boneName=spine_05}, planeTransform {location=<0 -10 0>, rotation=<rot 0 0 -90>}}]
    relativeSpaceBone {boneName=root}
    sphericalLimits [{limitRadius=15, drivingBone {boneName=spine_05}, sphereLocalOffset=<0 0 -5>}]
   }
   n20 = Anim Dynamics - Bone: dyn_hair_L_04(ComponentPose=@n19, Alpha=1, ExternalForce="0,0,0") {
    angularDampingOverride=0.999
    angularSpringConstant=10000
    bOverrideAngularDamping=true
    bUseGravityOverride=true
    bUsePlanarLimit=false
    bUseSphericalLimits=true
    gravityScale=0.35
    simulationSpace=BoneRelative
    alphaBoolBlend {blendOption=Linear}
    boundBone {boneName=dyn_hair_L_04}
    componentAppliedLinearAccClamp=<3000 3000 3000>
    componentLinearAccScale=<1 1 2>
    componentLinearVelScale=<1 1 2>
    gravityOverride {z=-980}
    physicsBodyDefinitions [{
      boundBone {boneName=dyn_hair_L_04}
      boxExtents=<10 40 40>
      constraintSetup {angularLimitsMax=<0 55 10>, angularLimitsMin=<0 -45 -10>, angularTarget=<1 0 0>}
      localJointOffset=<3 0 0>
     }]
    planarLimits [{drivingBone {boneName=spine_05}, planeTransform {location=<0 -10 0>, rotation=<rot 0 0 -90>}}]
    relativeSpaceBone {boneName=root}
    sphericalLimits [{limitRadius=15, drivingBone {boneName=spine_05}, sphereLocalOffset=<0 0 -5>}]
   }
   n21 = Anim Dynamics - Bone: dyn_hair_L_05(ComponentPose=@n20, Alpha=1, ExternalForce="0,0,0") {
    angularDampingOverride=0.999
    angularSpringConstant=10000
    bOverrideAngularDamping=true
    bUseGravityOverride=true
    bUsePlanarLimit=false
    bUseSphericalLimits=true
    gravityScale=0.35
    simulationSpace=BoneRelative
    alphaBoolBlend {blendOption=Linear}
    boundBone {boneName=dyn_hair_L_05}
    componentAppliedLinearAccClamp=<3000 3000 3000>
    componentLinearAccScale=<1 1 2>
    componentLinearVelScale=<1 1 2>
    gravityOverride {z=-980}
    physicsBodyDefinitions [{
      boundBone {boneName=dyn_hair_L_05}
      boxExtents=<10 40 40>
      constraintSetup {angularLimitsMax=<0 55 10>, angularLimitsMin=<0 -45 -10>, angularTarget=<1 0 0>}
      localJointOffset=<3 0 0>
     }]
    planarLimits [{drivingBone {boneName=spine_05}, planeTransform {location=<0 -10 0>, rotation=<rot 0 0 -90>}}]
    relativeSpaceBone {boneName=root}
    sphericalLimits [{limitRadius=15, drivingBone {boneName=spine_05}, sphereLocalOffset=<-2 -1 -5>}]
   }
   n22 = Anim Dynamics - Bone: dyn_hair_R_02(ComponentPose=@n21, Alpha=1, ExternalForce="0,0,0") {
    angularDampingOverride=0.9999
    angularSpringConstant=10000
    bAngularSpring=true
    bOverrideAngularDamping=true
    bUseGravityOverride=true
    bUsePlanarLimit=false
    bUseSphericalLimits=true
    gravityScale=0.35
    simulationSpace=BoneRelative
    alphaBoolBlend {blendOption=Linear}
    boundBone {boneName=dyn_hair_R_02}
    componentAppliedLinearAccClamp=<3000 3000 3000>
    componentLinearAccScale=<0.05 0.05 0.7>
    componentLinearVelScale=<0.05 0.05 0.7>
    gravityOverride {z=-980}
    physicsBodyDefinitions [{
      boundBone {boneName=dyn_hair_R_02}
      boxExtents=<10 40 40>
      constraintSetup {angularLimitsMax=<0 30 10>, angularLimitsMin=<0 -45 -10>, angularTarget=<1 0 0>}
      localJointOffset=<3 0 0>
     }]
    planarLimits [{drivingBone {boneName=spine_05}, planeTransform {location=<0 -10 0>, rotation=<rot 0 0 -90>}}]
    relativeSpaceBone {boneName=root}
    sphericalLimits [{limitRadius=18, drivingBone {boneName=spine_05}, sphereLocalOffset=<0 5 5>}]
   }
   n23 = Anim Dynamics - Bone: dyn_hair_R_03(ComponentPose=@n22, Alpha=1, ExternalForce="0,0,0") {
    angularDampingOverride=0.9999
    angularSpringConstant=10000
    bOverrideAngularDamping=true
    bUseGravityOverride=true
    bUsePlanarLimit=false
    bUseSphericalLimits=true
    gravityScale=0.35
    simulationSpace=BoneRelative
    alphaBoolBlend {blendOption=Linear}
    boundBone {boneName=dyn_hair_R_03}
    componentAppliedLinearAccClamp=<3000 3000 3000>
    componentLinearAccScale=<0.05 0.05 0.5>
    componentLinearVelScale=<0.05 0.05 0.5>
    gravityOverride {z=-980}
    physicsBodyDefinitions [{
      boundBone {boneName=dyn_hair_R_03}
      boxExtents=<10 40 40>
      constraintSetup {angularLimitsMax=<0 55 10>, angularLimitsMin=<0 -45 -10>, angularTarget=<1 0 0>}
      localJointOffset=<3 0 0>
     }]
    planarLimits [{drivingBone {boneName=spine_05}, planeTransform {location=<0 -10 0>, rotation=<rot 0 0 -90>}}]
    relativeSpaceBone {boneName=root}
    sphericalLimits [
     {limitRadius=15, drivingBone {boneName=spine_05}, sphereLocalOffset=<0 2 5>}
     {limitRadius=7, drivingBone {boneName=upperarm_r}, sphereLocalOffset=<0 0 0>}
    ]
   }
   n24 = Anim Dynamics - Bone: dyn_hair_R_04(ComponentPose=@n23, Alpha=1, ExternalForce="0,0,0") {
    angularDampingOverride=0.999
    angularSpringConstant=10000
    bOverrideAngularDamping=true
    bUseGravityOverride=true
    bUsePlanarLimit=false
    bUseSphericalLimits=true
    gravityScale=0.35
    simulationSpace=BoneRelative
    alphaBoolBlend {blendOption=Linear}
    boundBone {boneName=dyn_hair_R_04}
    componentAppliedLinearAccClamp=<3000 3000 3000>
    componentLinearAccScale=<0.5 0.5 1>
    componentLinearVelScale=<0.5 0.5 1>
    gravityOverride {z=-980}
    physicsBodyDefinitions [{
      boundBone {boneName=dyn_hair_R_04}
      boxExtents=<10 40 40>
      constraintSetup {angularLimitsMax=<0 55 10>, angularLimitsMin=<0 -45 -10>, angularTarget=<1 0 0>}
      localJointOffset=<3 0 0>
     }]
    planarLimits [{drivingBone {boneName=spine_05}, planeTransform {location=<0 -10 0>, rotation=<rot 0 0 -90>}}]
    relativeSpaceBone {boneName=root}
    sphericalLimits [
     {limitRadius=15, drivingBone {boneName=spine_05}, sphereLocalOffset=<0 2 5>}
     {limitRadius=8, drivingBone {boneName=upperarm_r}, sphereLocalOffset=<-5 0 0>}
    ]
   }
   n25 = Anim Dynamics - Bone: dyn_hair_R_05(ComponentPose=@n24, Alpha=1, ExternalForce="0,0,0") {
    angularDampingOverride=0.999
    angularSpringConstant=10000
    bOverrideAngularDamping=true
    bUseGravityOverride=true
    bUsePlanarLimit=false
    bUseSphericalLimits=true
    gravityScale=0.35
    simulationSpace=BoneRelative
    alphaBoolBlend {blendOption=Linear}
    boundBone {boneName=dyn_hair_R_05}
    componentAppliedLinearAccClamp=<3000 3000 3000>
    componentLinearAccScale=<0.5 0.5 1>
    componentLinearVelScale=<0.5 0.5 1>
    gravityOverride {z=-980}
    physicsBodyDefinitions [{
      boundBone {boneName=dyn_hair_R_05}
      boxExtents=<10 40 40>
      constraintSetup {angularLimitsMax=<0 55 10>, angularLimitsMin=<0 -45 -10>, angularTarget=<1 0 0>}
      localJointOffset=<3 0 0>
     }]
    planarLimits [{drivingBone {boneName=spine_05}, planeTransform {location=<0 -10 0>, rotation=<rot 0 0 -90>}}]
    relativeSpaceBone {boneName=root}
    sphericalLimits [
     {limitRadius=15, drivingBone {boneName=spine_05}, sphereLocalOffset=<-2 -1 5>}
     {limitRadius=8, drivingBone {boneName=upperarm_r}, sphereLocalOffset=<-9 0 0>}
    ]
   }
   n26 = Blend Poses by bool(
    BlendPose_0=Blend Poses by bool(BlendPose_0=Component To Local(ComponentPose=@n12), BlendPose_1=Blend Poses by bool(BlendPose_0=@n11, BlendPose_1=@n7) {blendType=Linear}) {blendType=Linear}
    BlendPose_1=Component To Local(ComponentPose=@n25)
   ) {blendType=Linear}
   n27 = Get Body_Mesh()
   n28 = RigidBody(
    ComponentPose=Local To Component(LocalPose=Copy Pose From Mesh(SourceMeshComponent=@n27) {bCopyCurves=true, bUseAttachedParent=true})
    OverrideWorldGravity=Get Gravity Override()
   ) {
    bClampLinearTranslationLimitToRefPose=true
    bOverrideWorldGravity=true
    overridePhysicsAsset=&F_MED_TimeInterval_FaceAcc_BackPack
    simulationSpace=BaseBoneSpace
    alphaBoolBlend {blendOption=Linear}
    alphaScaleBiasClamp {bInterpResult=true, interpSpeedIncreasing=2}
    baseBoneRef {boneName=spine_05}
    componentAppliedLinearAccClamp=<2000 2000 2000>
    componentLinearAccScale=<0.1 0.1 0.6>
    componentLinearVelScale=<0.1 0.1 0.6>
    simSpaceSettings {
     maxAngularAcceleration=1000
     maxAngularVelocity=3
     maxLinearAcceleration=700
     maxLinearVelocity=500
     velocityScaleZ=0.3
     worldAlpha=0.2
     externalLinearDragV=<0.2 0.2>
    }
   }
   n29 = "Constraint - dyn_hair_B_02 (PP)"!(ComponentPose=@n28, Alpha=1, ConstraintWeights_0=0.8, ConstraintWeights_1:float=0.2) -> (Pose) {
    alphaBoolBlend {blendOption=Linear}
    boneToModify {boneName=dyn_hair_B_02}
    constraintSetup [{targetBone {boneName=spine_05}}, {targetBone {boneName=head}}]
    constraintWeights [1,1]
   }
   n30 = "Constraint - dyn_hair_B_03 (P)"(ComponentPose=@n29, Alpha=1, ConstraintWeights_0=1) {
    alphaBoolBlend {blendOption=Linear}
    boneToModify {boneName=dyn_hair_B_03}
    constraintSetup [{targetBone {boneName=spine_05}}]
    constraintWeights [1]
   }
   n31 = "Constraint - dyn_hair_R_02 (PP)"!(ComponentPose=@n30, Alpha=1, ConstraintWeights_0=0.8, ConstraintWeights_1:float=0.2) -> (Pose) {
    alphaBoolBlend {blendOption=Linear}
    boneToModify {boneName=dyn_hair_R_02}
    constraintSetup [{targetBone {boneName=spine_05}}, {targetBone {boneName=spine_05}}]
    constraintWeights [1,1]
   }
   n32 = "Transform (Modify) Bone - Bone: dyn_hair_R_02"(ComponentPose=@n31, Alpha=1, Translation="0,0,3", Rotation="0,0,0", Scale="1,1,1") {
    rotationMode=BMM_Additive
    rotationSpace=BCS_BoneSpace
    scaleMode=BMM_Additive
    scaleSpace=BCS_BoneSpace
    translationMode=BMM_Additive
    translationSpace=BCS_BoneSpace
    alphaBoolBlend {blendOption=Linear}
    boneToModify {boneName=dyn_hair_R_02}
    translation {z=3}
   }
   n33 = "Constraint - dyn_hair_R_03 (PP)"!(ComponentPose=@n32, Alpha=1, ConstraintWeights_0=0.8, ConstraintWeights_1:float=0.2) -> (Pose) {
    alphaBoolBlend {blendOption=Linear}
    boneToModify {boneName=dyn_hair_R_03}
    constraintSetup [{targetBone {boneName=spine_05}}, {targetBone {boneName=clavicle_r}}]
    constraintWeights [1,1]
   }
   n34 = "Transform (Modify) Bone - Bone: dyn_hair_R_03"(ComponentPose=@n33, Alpha=1, Translation="-0.000,-0.000,1.637", Rotation="0,0,0", Scale="1,1,1") {
    rotationMode=BMM_Additive
    rotationSpace=BCS_BoneSpace
    scaleMode=BMM_Additive
    scaleSpace=BCS_BoneSpace
    translationMode=BMM_Additive
    translationSpace=BCS_BoneSpace
    alphaBoolBlend {blendOption=Linear}
    boneToModify {boneName=dyn_hair_R_03}
    translation=<0 0 1.6374>
   }
   n35 = "Constraint - dyn_hair_R_04 (PP)"!(ComponentPose=@n34, Alpha=1, ConstraintWeights_0=0.9, ConstraintWeights_1:float=0.1) -> (Pose) {
    alphaBoolBlend {blendOption=Linear}
    boneToModify {boneName=dyn_hair_R_04}
    constraintSetup [{targetBone {boneName=spine_05}}, {targetBone {boneName=clavicle_r}}]
    constraintWeights [1,1]
   }
   n36 = "Constraint - dyn_hair_L_02 (PPP)"!(ComponentPose=@n35, Alpha=1, ConstraintWeights_0=0.3, ConstraintWeights_1:float=0.2, ConstraintWeights_2:float=0.5) -> (Pose) {
    alphaBoolBlend {blendOption=Linear}
    boneToModify {boneName=dyn_hair_L_02}
    constraintSetup [{targetBone {boneName=spine_05}}, {targetBone {boneName=spine_05}}, {targetBone {boneName=clavicle_l}}]
    constraintWeights [1,1,1]
   }
   n37 = "Transform (Modify) Bone - Bone: dyn_hair_L_02"(ComponentPose=@n36, Alpha=1, Translation="3.034,-1.384,4.061", Rotation="-14.279,0.000,-0.000", Scale="1,1,1") {
    rotationMode=BMM_Additive
    rotationSpace=BCS_BoneSpace
    scaleMode=BMM_Additive
    scaleSpace=BCS_BoneSpace
    translationMode=BMM_Additive
    translationSpace=BCS_BoneSpace
    alphaBoolBlend {blendOption=Linear}
    boneToModify {boneName=dyn_hair_L_02}
    rotation {pitch=-14.279}
    translation=<3.034 -1.384 4.061>
   }
   n38 = "Constraint - dyn_hair_L_03 (PPP)"!(ComponentPose=@n37, Alpha=1, ConstraintWeights_0=0, ConstraintWeights_1:float=0.2, ConstraintWeights_2:float=0.8) -> (Pose) {
    alphaBoolBlend {blendOption=Linear}
    boneToModify {boneName=dyn_hair_L_03}
    constraintSetup [{targetBone {boneName=spine_05}}, {targetBone {boneName=head}}, {targetBone {boneName=clavicle_l}}]
    constraintWeights [1,1,1]
   }
   n39 = "Transform (Modify) Bone - Bone: dyn_hair_L_03"(ComponentPose=@n38, Alpha=1, Translation="-0.170,-2.720,6.488", Rotation="-25.026,0.000,-0.000", Scale="1.500,1.000,1.000") {
    rotationMode=BMM_Additive
    rotationSpace=BCS_BoneSpace
    scaleMode=BMM_Additive
    scaleSpace=BCS_BoneSpace
    translationMode=BMM_Additive
    translationSpace=BCS_BoneSpace
    alphaBoolBlend {blendOption=Linear}
    boneToModify {boneName=dyn_hair_L_03}
    rotation {pitch=-25.026}
    scale {x=1.5}
    translation=<-0.17 -2.72 6.488>
   }
   n40 = "Constraint - dyn_hair_L_04 (PP)"!(ComponentPose=@n39, Alpha=1, ConstraintWeights_0=0.7, ConstraintWeights_1:float=0.3) -> (Pose) {
    alphaBoolBlend {blendOption=Linear}
    boneToModify {boneName=dyn_hair_L_04}
    constraintSetup [{targetBone {boneName=spine_05}}, {targetBone {boneName=clavicle_l}}]
    constraintWeights [1,1]
   }
   n41 = "Transform (Modify) Bone - Bone: dyn_hair_L_04"(ComponentPose=@n40, Alpha=1, Translation="-1.528,-2.465,7.075", Rotation="-46.488,-19.883,18.997", Scale="1,1,1") {
    rotationMode=BMM_Additive
    rotationSpace=BCS_BoneSpace
    scaleSpace=BCS_BoneSpace
    translationMode=BMM_Additive
    translationSpace=BCS_BoneSpace
    alphaBoolBlend {blendOption=Linear}
    boneToModify {boneName=dyn_hair_L_04}
    rotation=<rot -46.488 -19.883 18.997>
    translation=<-1.528 -2.465 7.075>
   }
   n42 = "Constraint - dyn_hair_L_05 (PP)"!(ComponentPose=@n41, Alpha=1, ConstraintWeights_0=1, ConstraintWeights_1:float=0) -> (Pose) {
    alphaBoolBlend {blendOption=Linear}
    boneToModify {boneName=dyn_hair_L_05}
    constraintSetup [{targetBone {boneName=spine_05}}, {targetBone {boneName=clavicle_l}}]
    constraintWeights [1,1]
   }
   n43 = "Constraint - dyn_hair_SR_02 (PPP)"!(ComponentPose=@n42, Alpha=1, ConstraintWeights_0=0, ConstraintWeights_1:float=0, ConstraintWeights_2:float=1) -> (Pose) {
    alphaBoolBlend {blendOption=Linear}
    boneToModify {boneName=dyn_hair_SR_02}
    constraintSetup [{targetBone {boneName=spine_05}}, {targetBone {boneName=head}}, {targetBone {boneName=dyn_hair_F_02}}]
    constraintWeights [1,1,1]
   }
   n44 = "Constraint - dyn_hair_SR_03 (PPP)"!(ComponentPose=@n43, Alpha=1, ConstraintWeights_0=0, ConstraintWeights_1:float=1, ConstraintWeights_2:float=0) -> (Pose) {
    alphaBoolBlend {blendOption=Linear}
    boneToModify {boneName=dyn_hair_SR_03}
    constraintSetup [{targetBone {boneName=spine_05}}, {targetBone {boneName=clavicle_r}}, {targetBone {boneName=dyn_hair_F_02}}]
    constraintWeights [1,1,1]
   }
   n45 = "Constraint - dyn_hair_SL_02 (PPP)"!(ComponentPose=@n44, Alpha=1, ConstraintWeights_0=0, ConstraintWeights_1:float=0.1, ConstraintWeights_2:float=0.9) -> (Pose) {
    alphaBoolBlend {blendOption=Linear}
    boneToModify {boneName=dyn_hair_SL_02}
    constraintSetup [{targetBone {boneName=spine_05}}, {targetBone {boneName=clavicle_l}}, {targetBone {boneName=head}}]
    constraintWeights [1,1,1]
   }
   n46 = "Transform (Modify) Bone - Bone: dyn_hair_SL_02"(ComponentPose=@n45, Alpha=1, Translation="-2.528,-2.354,0.660", Rotation="-24.635,0.000,0.000", Scale="1,1,1") {
    rotationMode=BMM_Additive
    rotationSpace=BCS_BoneSpace
    translationMode=BMM_Additive
    translationSpace=BCS_BoneSpace
    alphaBoolBlend {blendOption=Linear}
    boneToModify {boneName=dyn_hair_SL_02}
    rotation=<rot -24.635 0 0>
    translation=<-2.528 -2.354 0.66>
   }
   n47 = "Constraint - dyn_hair_SL_03 (PPP)"!(ComponentPose=@n46, Alpha=1, ConstraintWeights_0=0.6, ConstraintWeights_1:float=0.2, ConstraintWeights_2:float=0.2) -> (Pose) {
    alphaBoolBlend {blendOption=Linear}
    boneToModify {boneName=dyn_hair_SL_03}
    constraintSetup [{targetBone {boneName=head}}, {targetBone {boneName=clavicle_l}}, {targetBone {boneName=dyn_hair_L_03}}]
    constraintWeights [1,1,1]
   }
   n48 = "Transform (Modify) Bone - Bone: dyn_hair_SL_03"(ComponentPose=@n47, Alpha=1, Translation="-1.835,-0.211,2.224", Rotation="-33.050,1.914,6.518", Scale="1,1,1") {
    rotationMode=BMM_Additive
    rotationSpace=BCS_BoneSpace
    translationMode=BMM_Additive
    translationSpace=BCS_BoneSpace
    alphaBoolBlend {blendOption=Linear}
    boneToModify {boneName=dyn_hair_SL_03}
    rotation=<rot -33.05 1.914 6.518>
    translation=<-1.835 -0.211 2.224>
   }
   n49 = Component To Local(ComponentPose=@n48)
   n50 = Get Body_Mesh()
   n51 = "Constraint - dyn_hair_B_02 (PP)"(
    ComponentPose=Local To Component(LocalPose=Copy Pose From Mesh(SourceMeshComponent=@n50) {bCopyCurves=true, bUseAttachedParent=true})
    Alpha=1
    ConstraintWeights_0=0.8
    ConstraintWeights_1=0.2
   ) {
    alphaBoolBlend {blendOption=Linear}
    boneToModify {boneName=dyn_hair_B_02}
    constraintSetup [{targetBone {boneName=spine_05}}, {targetBone {boneName=head}}]
    constraintWeights [1,1]
   }
   n52 = "Transform (Modify) Bone - Bone: dyn_hair_B_02"(ComponentPose=@n51, Alpha=1, Translation="0,0,5", Rotation="0,0,0", Scale="1,1,1") {
    rotationMode=BMM_Additive
    rotationSpace=BCS_BoneSpace
    scaleMode=BMM_Additive
    scaleSpace=BCS_BoneSpace
    translationMode=BMM_Additive
    translationSpace=BCS_BoneSpace
    alphaBoolBlend {blendOption=Linear}
    boneToModify {boneName=dyn_hair_B_02}
    translation {z=5}
   }
   n53 = "Constraint - dyn_hair_B_03 (P)"(ComponentPose=@n52, Alpha=1, ConstraintWeights_0=1) {
    alphaBoolBlend {blendOption=Linear}
    boneToModify {boneName=dyn_hair_B_03}
    constraintSetup [{targetBone {boneName=spine_05}}]
    constraintWeights [1]
   }
   n54 = "Transform (Modify) Bone - Bone: dyn_hair_B_03"(ComponentPose=@n53, Alpha=1, Translation="0,0,3", Rotation="5,0,0", Scale="1,1,1") {
    rotationMode=BMM_Additive
    rotationSpace=BCS_BoneSpace
    scaleMode=BMM_Additive
    scaleSpace=BCS_BoneSpace
    translationMode=BMM_Additive
    translationSpace=BCS_BoneSpace
    alphaBoolBlend {blendOption=Linear}
    boneToModify {boneName=dyn_hair_B_03}
    rotation {pitch=5}
    translation {z=3}
   }
   n55 = "Constraint - dyn_hair_R_02 (PP)"(ComponentPose=@n54, Alpha=1, ConstraintWeights_0=0.8, ConstraintWeights_1=0.2) {
    alphaBoolBlend {blendOption=Linear}
    boneToModify {boneName=dyn_hair_R_02}
    constraintSetup [{targetBone {boneName=spine_05}}, {targetBone {boneName=spine_05}}]
    constraintWeights [1,1]
   }
   n56 = "Transform (Modify) Bone - Bone: dyn_hair_R_02"(ComponentPose=@n55, Alpha=1, Translation="0,0,3", Rotation="0,0,0", Scale="1,1,1") {
    rotationMode=BMM_Additive
    rotationSpace=BCS_BoneSpace
    scaleMode=BMM_Additive
    scaleSpace=BCS_BoneSpace
    translationMode=BMM_Additive
    translationSpace=BCS_BoneSpace
    alphaBoolBlend {blendOption=Linear}
    boneToModify {boneName=dyn_hair_R_02}
    translation {z=3}
   }
   n57 = "Constraint - dyn_hair_R_03 (PP)"(ComponentPose=@n56, Alpha=1, ConstraintWeights_0=0.8, ConstraintWeights_1=0.2) {
    alphaBoolBlend {blendOption=Linear}
    boneToModify {boneName=dyn_hair_R_03}
    constraintSetup [{targetBone {boneName=spine_05}}, {targetBone {boneName=clavicle_r}}]
    constraintWeights [1,1]
   }
   n58 = "Transform (Modify) Bone - Bone: dyn_hair_R_03"(ComponentPose=@n57, Alpha=1, Translation="20,0,3", Rotation="0,0,0", Scale="1,1,1") {
    rotationMode=BMM_Additive
    rotationSpace=BCS_BoneSpace
    scaleMode=BMM_Additive
    scaleSpace=BCS_BoneSpace
    translationMode=BMM_Additive
    translationSpace=BCS_BoneSpace
    alphaBoolBlend {blendOption=Linear}
    boneToModify {boneName=dyn_hair_R_03}
    translation {x=20, z=3}
   }
   n59 = "Constraint - dyn_hair_R_04 (PP)"(ComponentPose=@n58, Alpha=1, ConstraintWeights_0=0.9, ConstraintWeights_1=0.1) {
    alphaBoolBlend {blendOption=Linear}
    boneToModify {boneName=dyn_hair_R_04}
    constraintSetup [{targetBone {boneName=spine_05}}, {targetBone {boneName=clavicle_r}}]
    constraintWeights [1,1]
   }
   n60 = "Transform (Modify) Bone - Bone: dyn_hair_R_04"(ComponentPose=@n59, Alpha=1, Translation="0,0,3", Rotation="0,0,0", Scale="1,1,1") {
    rotationMode=BMM_Additive
    rotationSpace=BCS_BoneSpace
    scaleMode=BMM_Additive
    scaleSpace=BCS_BoneSpace
    translationMode=BMM_Additive
    translationSpace=BCS_BoneSpace
    alphaBoolBlend {blendOption=Linear}
    boneToModify {boneName=dyn_hair_R_04}
    translation {z=3}
   }
   n61 = "Transform (Modify) Bone - Bone: dyn_hair_R_05"(ComponentPose=@n60, Alpha=1, Translation="0,0,3.5", Rotation="0,0,0", Scale="1,1,1") {
    rotationMode=BMM_Additive
    rotationSpace=BCS_BoneSpace
    scaleMode=BMM_Additive
    scaleSpace=BCS_BoneSpace
    translationMode=BMM_Additive
    translationSpace=BCS_BoneSpace
    alphaBoolBlend {blendOption=Linear}
    boneToModify {boneName=dyn_hair_R_05}
    translation {z=3.5}
   }
   n62 = "Constraint - dyn_hair_L_02 (PPP)"(ComponentPose=@n61, Alpha=1, ConstraintWeights_0=0.3, ConstraintWeights_1=0.2, ConstraintWeights_2=0.5) {
    alphaBoolBlend {blendOption=Linear}
    boneToModify {boneName=dyn_hair_L_02}
    constraintSetup [{targetBone {boneName=spine_05}}, {targetBone {boneName=spine_05}}, {targetBone {boneName=clavicle_l}}]
    constraintWeights [1,1,1]
   }
   n63 = "Transform (Modify) Bone - Bone: dyn_hair_L_02"(ComponentPose=@n62, Alpha=1, Translation="3.034,-1.384,4.061", Rotation="-14.279,0.000,-0.000", Scale="1,1,1") {
    rotationMode=BMM_Additive
    rotationSpace=BCS_BoneSpace
    scaleMode=BMM_Additive
    scaleSpace=BCS_BoneSpace
    translationMode=BMM_Additive
    translationSpace=BCS_BoneSpace
    alphaBoolBlend {blendOption=Linear}
    boneToModify {boneName=dyn_hair_L_02}
    rotation {pitch=-14.279}
    translation=<3.034 -1.384 4.061>
   }
   n64 = "Constraint - dyn_hair_L_03 (PPP)"(ComponentPose=@n63, Alpha=1, ConstraintWeights_0=0, ConstraintWeights_1=0.2, ConstraintWeights_2=0.8) {
    alphaBoolBlend {blendOption=Linear}
    boneToModify {boneName=dyn_hair_L_03}
    constraintSetup [{targetBone {boneName=spine_05}}, {targetBone {boneName=head}}, {targetBone {boneName=clavicle_l}}]
    constraintWeights [1,1,1]
   }
   n65 = "Transform (Modify) Bone - Bone: dyn_hair_L_03"(ComponentPose=@n64, Alpha=1, Translation="-0.170,-2.720,6.488", Rotation="-25.026,0.000,-0.000", Scale="1.500,1.000,1.000") {
    rotationMode=BMM_Additive
    rotationSpace=BCS_BoneSpace
    scaleMode=BMM_Additive
    scaleSpace=BCS_BoneSpace
    translationMode=BMM_Additive
    translationSpace=BCS_BoneSpace
    alphaBoolBlend {blendOption=Linear}
    boneToModify {boneName=dyn_hair_L_03}
    rotation {pitch=-25.026}
    scale {x=1.5}
    translation=<-0.17 -2.72 6.488>
   }
   n66 = "Constraint - dyn_hair_L_04 (PP)"(ComponentPose=@n65, Alpha=1, ConstraintWeights_0=0.7, ConstraintWeights_1=0.3) {
    alphaBoolBlend {blendOption=Linear}
    boneToModify {boneName=dyn_hair_L_04}
    constraintSetup [{targetBone {boneName=spine_05}}, {targetBone {boneName=clavicle_l}}]
    constraintWeights [1,1]
   }
   n67 = "Transform (Modify) Bone - Bone: dyn_hair_L_04"(ComponentPose=@n66, Alpha=1, Translation="-1.528,-2.465,7.075", Rotation="-46.488,-19.883,18.997", Scale="1,1,1") {
    rotationMode=BMM_Additive
    rotationSpace=BCS_BoneSpace
    scaleSpace=BCS_BoneSpace
    translationMode=BMM_Additive
    translationSpace=BCS_BoneSpace
    alphaBoolBlend {blendOption=Linear}
    boneToModify {boneName=dyn_hair_L_04}
    rotation=<rot -46.488 -19.883 18.997>
    translation=<-1.528 -2.465 7.075>
   }
   n68 = "Transform (Modify) Bone - Bone: dyn_hair_L_04"(ComponentPose=@n67, Alpha=1, Translation="0,0,3", Rotation="0,0,0", Scale="1,1,1") {
    rotationMode=BMM_Additive
    rotationSpace=BCS_BoneSpace
    scaleSpace=BCS_BoneSpace
    translationMode=BMM_Additive
    translationSpace=BCS_BoneSpace
    alphaBoolBlend {blendOption=Linear}
    boneToModify {boneName=dyn_hair_L_04}
    translation {z=3}
   }
   n69 = "Constraint - dyn_hair_L_05 (PP)"(ComponentPose=@n68, Alpha=1, ConstraintWeights_0=1, ConstraintWeights_1=0) {
    alphaBoolBlend {blendOption=Linear}
    boneToModify {boneName=dyn_hair_L_05}
    constraintSetup [{targetBone {boneName=spine_05}}, {targetBone {boneName=clavicle_l}}]
    constraintWeights [1,1]
   }
   n70 = "Transform (Modify) Bone - Bone: dyn_hair_L_05"(ComponentPose=@n69, Alpha=1, Translation="0,0,3.5", Rotation="0,0,0", Scale="1,1,1") {
    rotationMode=BMM_Additive
    rotationSpace=BCS_BoneSpace
    scaleSpace=BCS_BoneSpace
    translationMode=BMM_Additive
    translationSpace=BCS_BoneSpace
    alphaBoolBlend {blendOption=Linear}
    boneToModify {boneName=dyn_hair_L_05}
    translation {z=3.5}
   }
   n71 = "Constraint - dyn_hair_SR_02 (PPP)"(ComponentPose=@n70, Alpha=1, ConstraintWeights_0=0, ConstraintWeights_1=0, ConstraintWeights_2=1) {
    alphaBoolBlend {blendOption=Linear}
    boneToModify {boneName=dyn_hair_SR_02}
    constraintSetup [{targetBone {boneName=spine_05}}, {targetBone {boneName=head}}, {targetBone {boneName=dyn_hair_F_02}}]
    constraintWeights [1,1,1]
   }
   n72 = "Constraint - dyn_hair_SR_03 (PPP)"(ComponentPose=@n71, Alpha=1, ConstraintWeights_0=0, ConstraintWeights_1=1, ConstraintWeights_2=0) {
    alphaBoolBlend {blendOption=Linear}
    boneToModify {boneName=dyn_hair_SR_03}
    constraintSetup [{targetBone {boneName=spine_05}}, {targetBone {boneName=clavicle_r}}, {targetBone {boneName=dyn_hair_F_02}}]
    constraintWeights [1,1,1]
   }
   n73 = "Constraint - dyn_hair_SL_02 (PPP)"(ComponentPose=@n72, Alpha=1, ConstraintWeights_0=0, ConstraintWeights_1=0.1, ConstraintWeights_2=0.9) {
    alphaBoolBlend {blendOption=Linear}
    boneToModify {boneName=dyn_hair_SL_02}
    constraintSetup [{targetBone {boneName=spine_05}}, {targetBone {boneName=clavicle_l}}, {targetBone {boneName=head}}]
    constraintWeights [1,1,1]
   }
   n74 = "Transform (Modify) Bone - Bone: dyn_hair_SL_02"(ComponentPose=@n73, Alpha=1, Translation="-2.528,-2.354,0.660", Rotation="-24.635,0.000,0.000", Scale="1,1,1") {
    rotationMode=BMM_Additive
    rotationSpace=BCS_BoneSpace
    translationMode=BMM_Additive
    translationSpace=BCS_BoneSpace
    alphaBoolBlend {blendOption=Linear}
    boneToModify {boneName=dyn_hair_SL_02}
    rotation=<rot -24.635 0 0>
    translation=<-2.528 -2.354 0.66>
   }
   n75 = "Constraint - dyn_hair_SL_03 (PPP)"(ComponentPose=@n74, Alpha=1, ConstraintWeights_0=0.6, ConstraintWeights_1=0.2, ConstraintWeights_2=0.2) {
    alphaBoolBlend {blendOption=Linear}
    boneToModify {boneName=dyn_hair_SL_03}
    constraintSetup [{targetBone {boneName=head}}, {targetBone {boneName=clavicle_l}}, {targetBone {boneName=dyn_hair_L_03}}]
    constraintWeights [1,1,1]
   }
   n76 = "Transform (Modify) Bone - Bone: dyn_hair_SL_03"(ComponentPose=@n75, Alpha=1, Translation="-1.835,-0.211,2.224", Rotation="-33.050,1.914,6.518", Scale="1,1,1") {
    rotationMode=BMM_Additive
    rotationSpace=BCS_BoneSpace
    translationMode=BMM_Additive
    translationSpace=BCS_BoneSpace
    alphaBoolBlend {blendOption=Linear}
    boneToModify {boneName=dyn_hair_SL_03}
    rotation=<rot -33.05 1.914 6.518>
    translation=<-1.835 -0.211 2.224>
   }
   n77 = "Transform (Modify) Bone - Bone: dyn_hair_F_mstr"(ComponentPose=@n76, Alpha=1, Translation="0,0,0", Rotation="0,-10,0", Scale="1,1,1") {
    rotationMode=BMM_Additive
    rotationSpace=BCS_BoneSpace
    translationMode=BMM_Additive
    translationSpace=BCS_BoneSpace
    alphaBoolBlend {blendOption=Linear}
    boneToModify {boneName=dyn_hair_F_mstr}
    rotation {yaw=-10}
   }
   n78 = "Transform (Modify) Bone - Bone: dyn_hair_F_02"(ComponentPose=@n77, Alpha=1, Translation="0,0,0", Rotation="0,-20,0", Scale="1,1,1") {
    rotationMode=BMM_Additive
    rotationSpace=BCS_BoneSpace
    translationMode=BMM_Additive
    translationSpace=BCS_BoneSpace
    alphaBoolBlend {blendOption=Linear}
    boneToModify {boneName=dyn_hair_F_02}
    rotation {yaw=-20}
   }
   n79 = "Transform (Modify) Bone - Bone: dyn_hair_F_03"(ComponentPose=@n78, Alpha=1, Translation="0,0,0", Rotation="0,-20,0", Scale="1,1,1") {
    rotationMode=BMM_Additive
    rotationSpace=BCS_BoneSpace
    translationMode=BMM_Additive
    translationSpace=BCS_BoneSpace
    alphaBoolBlend {blendOption=Linear}
    boneToModify {boneName=dyn_hair_F_03}
    rotation {yaw=-20}
   }
   n80 = "Transform (Modify) Bone - Bone: dyn_hair_F_mstr"(ComponentPose=@n79, Alpha=1, Translation="0,0,0", Rotation="-10,-10,0", Scale="1,1,1") {
    rotationMode=BMM_Additive
    rotationSpace=BCS_BoneSpace
    translationMode=BMM_Additive
    translationSpace=BCS_BoneSpace
    alphaBoolBlend {blendOption=Linear}
    boneToModify {boneName=dyn_hair_F_mstr}
    rotation {pitch=-10, yaw=-10}
   }
   n81 = "Transform (Modify) Bone - Bone: dyn_hair_F_02"(ComponentPose=@n80, Alpha=1, Translation="0,0,0", Rotation="-20,-25,0", Scale="1,1,1") {
    rotationMode=BMM_Additive
    rotationSpace=BCS_BoneSpace
    translationMode=BMM_Additive
    translationSpace=BCS_BoneSpace
    alphaBoolBlend {blendOption=Linear}
    boneToModify {boneName=dyn_hair_F_02}
    rotation {pitch=-20, yaw=-25}
   }
   n82 = "Transform (Modify) Bone - Bone: dyn_hair_F_03"(ComponentPose=@n81, Alpha=1, Translation="0,0,0", Rotation="-20,-25,0", Scale="1,1,1") {
    rotationMode=BMM_Additive
    rotationSpace=BCS_BoneSpace
    translationMode=BMM_Additive
    translationSpace=BCS_BoneSpace
    alphaBoolBlend {blendOption=Linear}
    boneToModify {boneName=dyn_hair_F_03}
    rotation {pitch=-20, yaw=-25}
   }
   n83 = Component To Local(ComponentPose=@n82)
   n84 = Get Body_Mesh()
   n85 = "Constraint - dyn_hair_B_02 (PP)"(
    ComponentPose=Local To Component(LocalPose=Copy Pose From Mesh(SourceMeshComponent=@n84) {bCopyCurves=true, bUseAttachedParent=true})
    Alpha=1
    ConstraintWeights_0=0.8
    ConstraintWeights_1=0.2
   ) {
    alphaBoolBlend {blendOption=Linear}
    boneToModify {boneName=dyn_hair_B_02}
    constraintSetup [{targetBone {boneName=spine_05}}, {targetBone {boneName=head}}]
    constraintWeights [1,1]
   }
   n86 = "Constraint - dyn_hair_B_03 (P)"(ComponentPose=@n85, Alpha=1, ConstraintWeights_0=1) {
    alphaBoolBlend {blendOption=Linear}
    boneToModify {boneName=dyn_hair_B_03}
    constraintSetup [{targetBone {boneName=spine_05}}]
    constraintWeights [1]
   }
   n87 = "Constraint - dyn_hair_R_02 (PP)"(ComponentPose=@n86, Alpha=1, ConstraintWeights_0=0.8, ConstraintWeights_1=0.2) {
    alphaBoolBlend {blendOption=Linear}
    boneToModify {boneName=dyn_hair_R_02}
    constraintSetup [{targetBone {boneName=spine_05}}, {targetBone {boneName=spine_05}}]
    constraintWeights [1,1]
   }
   n88 = "Transform (Modify) Bone - Bone: dyn_hair_R_02"(ComponentPose=@n87, Alpha=1, Translation="0,0,3", Rotation="0,0,0", Scale="1,1,1") {
    rotationMode=BMM_Additive
    rotationSpace=BCS_BoneSpace
    scaleMode=BMM_Additive
    scaleSpace=BCS_BoneSpace
    translationMode=BMM_Additive
    translationSpace=BCS_BoneSpace
    alphaBoolBlend {blendOption=Linear}
    boneToModify {boneName=dyn_hair_R_02}
    translation {z=3}
   }
   n89 = "Constraint - dyn_hair_R_03 (PP)"(ComponentPose=@n88, Alpha=1, ConstraintWeights_0=0.8, ConstraintWeights_1=0.2) {
    alphaBoolBlend {blendOption=Linear}
    boneToModify {boneName=dyn_hair_R_03}
    constraintSetup [{targetBone {boneName=spine_05}}, {targetBone {boneName=clavicle_r}}]
    constraintWeights [1,1]
   }
   n90 = "Transform (Modify) Bone - Bone: dyn_hair_R_03"(ComponentPose=@n89, Alpha=1, Translation="-0.000,-0.000,1.637", Rotation="0,0,0", Scale="1,1,1") {
    rotationMode=BMM_Additive
    rotationSpace=BCS_BoneSpace
    scaleMode=BMM_Additive
    scaleSpace=BCS_BoneSpace
    translationMode=BMM_Additive
    translationSpace=BCS_BoneSpace
    alphaBoolBlend {blendOption=Linear}
    boneToModify {boneName=dyn_hair_R_03}
    translation {z=1.6374}
   }
   n91 = "Constraint - dyn_hair_R_04 (PP)"(ComponentPose=@n90, Alpha=1, ConstraintWeights_0=0.9, ConstraintWeights_1=0.1) {
    alphaBoolBlend {blendOption=Linear}
    boneToModify {boneName=dyn_hair_R_04}
    constraintSetup [{targetBone {boneName=spine_05}}, {targetBone {boneName=clavicle_r}}]
    constraintWeights [1,1]
   }
   n92 = "Constraint - dyn_hair_L_02 (PPP)"(ComponentPose=@n91, Alpha=1, ConstraintWeights_0=0.3, ConstraintWeights_1=0.2, ConstraintWeights_2=0.5) {
    alphaBoolBlend {blendOption=Linear}
    boneToModify {boneName=dyn_hair_L_02}
    constraintSetup [{targetBone {boneName=spine_05}}, {targetBone {boneName=spine_05}}, {targetBone {boneName=clavicle_l}}]
    constraintWeights [1,1,1]
   }
   n93 = "Transform (Modify) Bone - Bone: dyn_hair_L_02"(ComponentPose=@n92, Alpha=1, Translation="3.034,-1.384,4.061", Rotation="-14.279,0.000,-0.000", Scale="1,1,1") {
    rotationMode=BMM_Additive
    rotationSpace=BCS_BoneSpace
    scaleMode=BMM_Additive
    scaleSpace=BCS_BoneSpace
    translationMode=BMM_Additive
    translationSpace=BCS_BoneSpace
    alphaBoolBlend {blendOption=Linear}
    boneToModify {boneName=dyn_hair_L_02}
    rotation {pitch=-14.279}
    translation=<3.034 -1.384 4.061>
   }
   n94 = "Constraint - dyn_hair_L_03 (PPP)"(ComponentPose=@n93, Alpha=1, ConstraintWeights_0=0, ConstraintWeights_1=0.2, ConstraintWeights_2=0.8) {
    alphaBoolBlend {blendOption=Linear}
    boneToModify {boneName=dyn_hair_L_03}
    constraintSetup [{targetBone {boneName=spine_05}}, {targetBone {boneName=head}}, {targetBone {boneName=clavicle_l}}]
    constraintWeights [1,1,1]
   }
   n95 = "Transform (Modify) Bone - Bone: dyn_hair_L_03"(ComponentPose=@n94, Alpha=1, Translation="-0.170,-2.720,6.488", Rotation="-25.026,0.000,-0.000", Scale="1.500,1.000,1.000") {
    rotationMode=BMM_Additive
    rotationSpace=BCS_BoneSpace
    scaleMode=BMM_Additive
    scaleSpace=BCS_BoneSpace
    translationMode=BMM_Additive
    translationSpace=BCS_BoneSpace
    alphaBoolBlend {blendOption=Linear}
    boneToModify {boneName=dyn_hair_L_03}
    rotation {pitch=-25.026}
    scale {x=1.5}
    translation=<-0.17 -2.72 6.488>
   }
   n96 = "Constraint - dyn_hair_L_04 (PP)"(ComponentPose=@n95, Alpha=1, ConstraintWeights_0=0.7, ConstraintWeights_1=0.3) {
    alphaBoolBlend {blendOption=Linear}
    boneToModify {boneName=dyn_hair_L_04}
    constraintSetup [{targetBone {boneName=spine_05}}, {targetBone {boneName=clavicle_l}}]
    constraintWeights [1,1]
   }
   n97 = "Transform (Modify) Bone - Bone: dyn_hair_L_04"(ComponentPose=@n96, Alpha=1, Translation="-1.528,-2.465,7.075", Rotation="-46.488,-19.883,18.997", Scale="1,1,1") {
    rotationMode=BMM_Additive
    rotationSpace=BCS_BoneSpace
    scaleSpace=BCS_BoneSpace
    translationMode=BMM_Additive
    translationSpace=BCS_BoneSpace
    alphaBoolBlend {blendOption=Linear}
    boneToModify {boneName=dyn_hair_L_04}
    rotation=<rot -46.488 -19.883 18.997>
    translation=<-1.528 -2.465 7.075>
   }
   n98 = "Constraint - dyn_hair_L_05 (PP)"(ComponentPose=@n97, Alpha=1, ConstraintWeights_0=1, ConstraintWeights_1=0) {
    alphaBoolBlend {blendOption=Linear}
    boneToModify {boneName=dyn_hair_L_05}
    constraintSetup [{targetBone {boneName=spine_05}}, {targetBone {boneName=clavicle_l}}]
    constraintWeights [1,1]
   }
   n99 = "Constraint - dyn_hair_SR_02 (PPP)"(ComponentPose=@n98, Alpha=1, ConstraintWeights_0=0, ConstraintWeights_1=0, ConstraintWeights_2=1) {
    alphaBoolBlend {blendOption=Linear}
    boneToModify {boneName=dyn_hair_SR_02}
    constraintSetup [{targetBone {boneName=spine_05}}, {targetBone {boneName=head}}, {targetBone {boneName=dyn_hair_F_02}}]
    constraintWeights [1,1,1]
   }
   n100 = "Constraint - dyn_hair_SR_03 (PPP)"(ComponentPose=@n99, Alpha=1, ConstraintWeights_0=0, ConstraintWeights_1=1, ConstraintWeights_2=0) {
    alphaBoolBlend {blendOption=Linear}
    boneToModify {boneName=dyn_hair_SR_03}
    constraintSetup [{targetBone {boneName=spine_05}}, {targetBone {boneName=clavicle_r}}, {targetBone {boneName=dyn_hair_F_02}}]
    constraintWeights [1,1,1]
   }
   n101 = "Constraint - dyn_hair_SL_02 (PPP)"(ComponentPose=@n100, Alpha=1, ConstraintWeights_0=0, ConstraintWeights_1=0.1, ConstraintWeights_2=0.9) {
    alphaBoolBlend {blendOption=Linear}
    boneToModify {boneName=dyn_hair_SL_02}
    constraintSetup [{targetBone {boneName=spine_05}}, {targetBone {boneName=clavicle_l}}, {targetBone {boneName=head}}]
    constraintWeights [1,1,1]
   }
   n102 = "Transform (Modify) Bone - Bone: dyn_hair_SL_02"(ComponentPose=@n101, Alpha=1, Translation="-2.528,-2.354,0.660", Rotation="-24.635,0.000,0.000", Scale="1,1,1") {
    rotationMode=BMM_Additive
    rotationSpace=BCS_BoneSpace
    translationMode=BMM_Additive
    translationSpace=BCS_BoneSpace
    alphaBoolBlend {blendOption=Linear}
    boneToModify {boneName=dyn_hair_SL_02}
    rotation=<rot -24.635 0 0>
    translation=<-2.528 -2.354 0.66>
   }
   n103 = "Constraint - dyn_hair_SL_03 (PPP)"(ComponentPose=@n102, Alpha=1, ConstraintWeights_0=0.6, ConstraintWeights_1=0.2, ConstraintWeights_2=0.2) {
    alphaBoolBlend {blendOption=Linear}
    boneToModify {boneName=dyn_hair_SL_03}
    constraintSetup [{targetBone {boneName=head}}, {targetBone {boneName=clavicle_l}}, {targetBone {boneName=dyn_hair_L_03}}]
    constraintWeights [1,1,1]
   }
   n104 = "Transform (Modify) Bone - Bone: dyn_hair_SL_03"(ComponentPose=@n103, Alpha=1, Translation="-1.835,-0.211,2.224", Rotation="-33.050,1.914,6.518", Scale="1,1,1") {
    rotationMode=BMM_Additive
    rotationSpace=BCS_BoneSpace
    translationMode=BMM_Additive
    translationSpace=BCS_BoneSpace
    alphaBoolBlend {blendOption=Linear}
    boneToModify {boneName=dyn_hair_SL_03}
    rotation=<rot -33.05 1.914 6.518>
    translation=<-1.835 -0.211 2.224>
   }
   n105 = "Transform (Modify) Bone - Bone: dyn_hair_F_mstr"(ComponentPose=@n104, Alpha=1, Translation="0,0,0", Rotation="0,-10,0", Scale="1,1,1") {
    rotationMode=BMM_Additive
    rotationSpace=BCS_BoneSpace
    translationMode=BMM_Additive
    translationSpace=BCS_BoneSpace
    alphaBoolBlend {blendOption=Linear}
    boneToModify {boneName=dyn_hair_F_mstr}
    rotation {yaw=-10}
   }
   n106 = "Transform (Modify) Bone - Bone: dyn_hair_F_02"(ComponentPose=@n105, Alpha=1, Translation="0,0,0", Rotation="0,-20,0", Scale="1,1,1") {
    rotationMode=BMM_Additive
    rotationSpace=BCS_BoneSpace
    translationMode=BMM_Additive
    translationSpace=BCS_BoneSpace
    alphaBoolBlend {blendOption=Linear}
    boneToModify {boneName=dyn_hair_F_02}
    rotation {yaw=-20}
   }
   n107 = "Transform (Modify) Bone - Bone: dyn_hair_F_03"(ComponentPose=@n106, Alpha=1, Translation="0,0,0", Rotation="0,-20,0", Scale="1,1,1") {
    rotationMode=BMM_Additive
    rotationSpace=BCS_BoneSpace
    translationMode=BMM_Additive
    translationSpace=BCS_BoneSpace
    alphaBoolBlend {blendOption=Linear}
    boneToModify {boneName=dyn_hair_F_03}
    rotation {yaw=-20}
   }
   n108 = "Transform (Modify) Bone - Bone: dyn_hair_F_mstr"(ComponentPose=@n107, Alpha=1, Translation="0,0,0", Rotation="-10,-10,0", Scale="1,1,1") {
    rotationMode=BMM_Additive
    rotationSpace=BCS_BoneSpace
    translationMode=BMM_Additive
    translationSpace=BCS_BoneSpace
    alphaBoolBlend {blendOption=Linear}
    boneToModify {boneName=dyn_hair_F_mstr}
    rotation {pitch=-10, yaw=-10}
   }
   n109 = "Transform (Modify) Bone - Bone: dyn_hair_F_02"(ComponentPose=@n108, Alpha=1, Translation="0,0,0", Rotation="-20,-25,0", Scale="1,1,1") {
    rotationMode=BMM_Additive
    rotationSpace=BCS_BoneSpace
    translationMode=BMM_Additive
    translationSpace=BCS_BoneSpace
    alphaBoolBlend {blendOption=Linear}
    boneToModify {boneName=dyn_hair_F_02}
    rotation {pitch=-20, yaw=-25}
   }
   n110 = "Transform (Modify) Bone - Bone: dyn_hair_F_03"(ComponentPose=@n109, Alpha=1, Translation="0,0,0", Rotation="-20,-25,0", Scale="1,1,1") {
    rotationMode=BMM_Additive
    rotationSpace=BCS_BoneSpace
    translationMode=BMM_Additive
    translationSpace=BCS_BoneSpace
    alphaBoolBlend {blendOption=Linear}
    boneToModify {boneName=dyn_hair_F_03}
    rotation {pitch=-20, yaw=-25}
   }
   n111 = Component To Local(ComponentPose=@n110)
   Output Pose(Result=Blend Poses by bool(
     BlendPose_0=Blend Poses by bool(BlendPose_0=@n83, BlendPose_1=@n111) {blendType=Linear}
     BlendPose_1=Blend Poses by bool(BlendPose_0=@n49, BlendPose_1=@n26) {blendType=Linear}
    ) {blendType=Linear}) {layerGroup=None}
   n112 = Get Body_Mesh()
   n113 = Control Rig!(
    Source=Layered blend per bone(
     BasePose=Copy Pose From Mesh() {bUseAttachedParent=true}
     BlendPoses_0=Copy Pose From Mesh(SourceMeshComponent=@n112) {bCopyCurves=true, bUseAttachedParent=true}
    ) {layerSetup [{branchFilters [{boneName=upperarm_l}, {boneName=upperarm_r}]}]}
    Alpha
    IsCrouchCtrl:bool=Get bIsCrouching()
    IsWearingCapeCtrl:bool=Get isWearingCape() -> (isWearingCape:bool)
    IsSkyDiveCtrl:bool=Get bIsSkyDiving()
   ) -> (Pose) {
    alphaBoolBlend {blendOption=Linear}
    controlRigAssetReference {blueprintRigClass=&/F_MED_TimeInterval_FaceAcc_CtrlRig.F_MED_TimeInterval_FaceAcc_CtrlRig_C}
   }
   Save cached pose 'PostCtrlRigCache'(Pose=@n113) {cacheName=PostCtrlRigCache}
   Get isWearingBackpack() -> (isWearingBackpack:bool)
  ]}
 EventGraph {nodes [
   Event BlueprintUpdateAnimation()
   > n1 = Sequence!(execute) -> (then_0, then_1, then_2:Exec, then_3:Exec, then_4:Exec, then_5:Exec)
   > n2 = Gravity Override State Function DEPRECATED(CharacterPartAnimInstance=Self-Reference(), GameTimeInSeconds=GetGameTimeInSeconds(), GravityOverrideParams=Get Gravity Override Params())
   > Set Gravity Override(Gravity Override=@n2)
   Set Body_Mesh(execute=@n1.then_1, Body_Mesh=GetOwnerSkeletalMeshComponent())
   Set SkyDiveAlpha(execute=@n1.then_2, SkyDiveAlpha:double=SelectFloat(A=0, B=1, bPickA=Get bIsSkyDiving())) -> (Output_Get:double)
   n3 = BreakTransform(InTransform=GetDeltaTransformFromRefPose(self=GetOwnerSkeletalMeshComponent(), BoneName=upperarm_l, BaseName=spine_05))
   Set ClavLAlpha(execute=@n1.then_3, ClavLAlpha:double=MapRangeClamped(Value=BreakRotator(InRot=@n3.Rotation).Pitch, InRangeA=0, InRangeB=70)) -> (Output_Get:double)
   n4 = BreakTransform(InTransform=GetDeltaTransformFromRefPose(self=GetOwnerSkeletalMeshComponent(), BoneName=upperarm_r, BaseName=spine_05))
   > Set ClavRAlpha(ClavRAlpha:double=MapRangeClamped(Value=BreakRotator(InRot=@n4.Rotation).Pitch, InRangeA=0, InRangeB=70)) -> (Output_Get:double)
   n5 = AND Boolean(
    A="Not Equal (Boolean)"(A=Get bRigidBodyEnabled(), B=true) -> (ReturnValue=~)
    B="Not Equal (Boolean)"(A=Get bAnimDynamicsEnabled() -> (bAnimDynamicsEnabled:bool), B=true) -> (ReturnValue=~)
   )
   n6 = Get isWearingBackpack()
   > Set HairPinAlpha(HairPinAlpha:double=SelectFloat(A=1, B=0, bPickA=OR Boolean(A=@n6, B=@n5))) -> (Output_Get:double)
   n7 = BreakTransform(InTransform=GetDeltaTransformFromRefPose(self=GetOwnerSkeletalMeshComponent(), BoneName=head, BaseName=spine_05))
   > Set NeckDelta(NeckDelta:double=MapRangeClamped(Value=BreakRotator(InRot=@n7.Rotation).Yaw, InRangeA=-15, InRangeB=25)) -> (Output_Get:double)
   n8 = BreakTransform(InTransform=GetDeltaTransformFromRefPose(self=GetOwnerSkeletalMeshComponent(), BoneName=head, BaseName=spine_05))
   > Set NeckR/LRotation(NeckR/LRotation:double=MapRangeClamped(Value=BreakRotator(InRot=@n8.Rotation), InRangeA=-10, InRangeB=-80)) -> (Output_Get:double)
   n9 = Get bRigidBodyEnabled()
   n10 = Get bAnimDynamicsEnabled()
   Set ASW_Cape(execute=@n1.then_4, ASW_Cape:double=SelectFloat(A=0, B=1, bPickA=AND Boolean(A=@n10, B=@n9))) -> (Output_Get:double)
   n11 = Get bAnimDynamicsEnabled()
   n12 = Get bRigidBodyEnabled()
   Set NoRbanOrAnim(
    execute=@n1.then_5
    NoRbanOrAnim:bool=AND Boolean(A="Not Equal (Boolean)"(A=@n12, B=true) -> (ReturnValue=~), B="Not Equal (Boolean)"(A=@n11, B=true) -> (ReturnValue=~))
   ) -> (Output_Get:bool)
   Event OnCharacterCustomizationLoaded()
   > Set CapeSpecific(CapeSpecific:bool=BackpackItemDefinitionHasAnyMetaTag!(Tags:Array of Gameplay Tag Structures=Make Array("[0]"=(TagName="Cosmetics.MetaTags.TimeIntervalCape")), bRequiresEquipped:bool=false) -> (ReturnValue:bool)) -> (Output_Get:bool)
   > n13 = Set SkelMesh(SkelMesh=GetPartSkeletalMeshComponent())
   > n14 = Get Dynamic Material Instance(SkeletalMeshComponent=@n13)
   > Set Mat Instance Dynamic(Mat Instance Dynamic=@n14)
   > n15 = GetOrCreateDynamicOverlayMaterial(MeshComponent=@n13, MaterialName=None)
   > Set Overlay Mat Instance Dynamic(Overlay Mat Instance Dynamic=@n15)
   n16 = CheckForBackbling()
   n17 = Get CapeSpecific() -> (CapeSpecific:bool)
   > Set isWearingBackpack(isWearingBackpack:bool=AND Boolean(
     A=OR Boolean(A=@n16, B=@n16.bIsBackblingACape)
     B="Equal (Boolean)"(A:bool=@n17, B:bool) -> (ReturnValue:bool=~) {is=title K2Node_PromotableOperator}
    )) -> (Output_Get:bool)
   n18 = Get CapeSpecific()
   > Set isWearingCape(isWearingCape=AND Boolean(A=AND Boolean(A=@n16.bIsBackblingLong, B=@n16), B=NOT Boolean(A=@n18)))
   GetPartSkeletalMeshComponent()
   PrintString!(execute, InString="To String (Boolean)"(InBool=false), bPrintToScreen, bPrintToLog, TextColor, Duration, Key) -> (then)
  ]}
]
variables [
 ASW_Cape {type=number}
 Body_Mesh {type=SkeletalMeshComponent}
 CapeSpecific {type=boolean}
 ClavLAlpha {type=number}
 ClavRAlpha {type=number}
 Gravity Override {type=Vector}
 Gravity Override Params {
  type=GravityOverrideParamsStruct
  default {skydive_DivingDown {gravityOverride=<-1500 0 400>}, skydive_FloatingDown {gravityOverride=<0 -200 200>}}
 }
 HairPinAlpha {type=number}
 IsCrouchingBol {type=boolean}
 isWearingBackpack {type=boolean}
 isWearingCape {type=boolean}
 Mat Instance Dynamic {type=MaterialInstanceDynamic}
 NeckDelta {type=number}
 NeckR/LRotation {type=number}
 NoRbanOrAnim {type=boolean}
 Overlay Mat Instance Dynamic {type=MaterialInstanceDynamic}
 SkelMesh {type=SkeletalMeshComponent}
 SkyDiveAlpha {type=number}
]
